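# Control Rig for the female HydraTrumpet CoachSalve body that adds corrective motion to belt bags, buckles, shoulder pads and left knee pad.
# After optional debug controls, it runs the shared DeformRig function, then pins belt pieces with weighted multi-bone parent constraints.
# DeformRig alpha attributes weight additional knee, shoulder and belt offsets.
# A left-clavicle spherical pose reader, scaled by "is not Emoting", adjusts that pad; crouching adds smoothed belt-bag offsets.
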ControlRigBlueprint /BRCosmetics/Characters/Player/Female/Medium/Bodies/F_MED_HydraTrumpet_CoachSalve/Meshes/F_MED_HydraTrumpet_CoachSalve_CtrlRig
graphs [
 New Function_ContainedGraph {nodes [
   RigVMFunctionEntryNode()
   > &RigUnit_ParentConstraint(
    AdvancedSettings=(InterpolationType=Average,RotationOrderForFilter=XZY)
    Child=(Type=Bone,Name="base_beltBags_01_L")
    Parents="((Item=(Type=Bone,Name=\"pelvis\"),Weight=0.7243),(Item=(Type=Bone,Name=\"deform_groin_l\"),Weight=0.1673),(Item=(Type=Bone,Name=\"spine_02\"),Weight=0.081),(Item=(Type=Bone,Name=\"thigh_twist_01_l\"),Weight=0.0187),(Item=(Type=Bone,Name=\"spine_04\"),Weight=0.0065) ...(truncated)"
    Weight=1
   )
   > &RigUnit_ParentConstraint(
    AdvancedSettings=(InterpolationType=Average,RotationOrderForFilter=XZY)
    Child=(Type=Bone,Name="base_beltBags_02_L")
    Parents="((Item=(Type=Bone,Name=\"pelvis\"),Weight=0.7236),(Item=(Type=Bone,Name=\"deform_groin_l\"),Weight=0.1637),(Item=(Type=Bone,Name=\"spine_02\"),Weight=0.0713),(Item=(Type=Bone,Name=\"thigh_twist_01_l\"),Weight=0.036),(Item=(Type=Bone,Name=\"spine_04\"),Weight=0.0048) ...(truncated)"
    Weight=1
   )
   > &RigUnit_ParentConstraint(
    AdvancedSettings=(InterpolationType=Average,RotationOrderForFilter=XZY)
    Child=(Type=Bone,Name="base_beltBags_03_L")
    Parents="((Item=(Type=Bone,Name=\"pelvis\"),Weight=0.8497),(Item=(Type=Bone,Name=\"spine_02\"),Weight=0.0554),(Item=(Type=Bone,Name=\"deform_groin_l\"),Weight=0.0528),(Item=(Type=Bone,Name=\"thigh_twist_01_l\"),Weight=0.034),(Item=(Type=Bone,Name=\"spine_01\"),Weight=0.0033) ...(truncated)"
    Weight=1
   )
   > &RigUnit_ParentConstraint(
    AdvancedSettings=(InterpolationType=Average,RotationOrderForFilter=XZY)
    Child=(Type=Bone,Name="base_beltBags_04_L")
    Parents=((Item=(Type=Bone,Name="pelvis"),Weight=0.9367),(Item=(Type=Bone,Name="spine_02"),Weight=0.0404),(Item=(Type=Bone,Name="thigh_twist_01_l"),Weight=0.0198),(Item=(Type=Bone,Name="spine_01"),Weight=0.003))
    Weight=1
   )
   > &RigUnit_ParentConstraint(
    AdvancedSettings=(InterpolationType=Average,RotationOrderForFilter=XZY)
    Child=(Type=Bone,Name="base_beltBags_05_L")
    Parents=((Item=(Type=Bone,Name="pelvis"),Weight=0.931),(Item=(Type=Bone,Name="spine_02"),Weight=0.0473),(Item=(Type=Bone,Name="spine_01"),Weight=0.0216))
    Weight=1
   )
   > &RigUnit_ParentConstraint(
    AdvancedSettings=(InterpolationType=Average,RotationOrderForFilter=XZY)
    Child=(Type=Bone,Name="base_beltBuckle_B")
    Parents=((Item=(Type=Bone,Name="pelvis"),Weight=0.8645),(Item=(Type=Bone,Name="spine_02"),Weight=0.0817),(Item=(Type=Bone,Name="spine_01"),Weight=0.0452),(Item=(Type=Bone,Name="spine_03"),Weight=0.0074),(Item=(Type=Bone,Name="deform_glute_r"),Weight=0.0012))
    Weight=1
   )
   > &RigUnit_ParentConstraint(
    AdvancedSettings=(InterpolationType=Average,RotationOrderForFilter=XZY)
    Child=(Type=Bone,Name="base_beltBags_05_R")
    Parents=((Item=(Type=Bone,Name="pelvis"),Weight=0.93),(Item=(Type=Bone,Name="spine_02"),Weight=0.048),(Item=(Type=Bone,Name="spine_01"),Weight=0.0221))
    Weight=1
   )
   > &RigUnit_ParentConstraint(
    AdvancedSettings=(InterpolationType=Average,RotationOrderForFilter=XZY)
    Child=(Type=Bone,Name="base_beltBags_04_R")
    Parents="((Item=(Type=Bone,Name=\"pelvis\"),Weight=0.918),(Item=(Type=Bone,Name=\"spine_02\"),Weight=0.0411),(Item=(Type=Bone,Name=\"thigh_twist_01_r\"),Weight=0.0189),(Item=(Type=Bone,Name=\"deform_glute_r\"),Weight=0.0183),(Item=(Type=Bone,Name=\"spine_01\"),Weight=0.0037) ...(truncated)"
    Weight=1
   )
   > &RigUnit_ParentConstraint(
    AdvancedSettings=(InterpolationType=Average,RotationOrderForFilter=XZY)
    Child=(Type=Bone,Name="base_beltBags_03_R")
    Parents="((Item=(Type=Bone,Name=\"pelvis\"),Weight=0.9296),(Item=(Type=Bone,Name=\"spine_02\"),Weight=0.0374),(Item=(Type=Bone,Name=\"thigh_twist_01_r\"),Weight=0.0305),(Item=(Type=Bone,Name=\"spine_01\"),Weight=0.002),(Item=(Type=Bone,Name=\"spine_04\"),Weight=0.0003),(Item ...(truncated)"
    Weight=1
   )
   > &RigUnit_ParentConstraint(
    AdvancedSettings=(InterpolationType=Average,RotationOrderForFilter=XZY)
    Child=(Type=Bone,Name="base_beltBags_02_R")
    Parents="((Item=(Type=Bone,Name=\"pelvis\"),Weight=0.7145),(Item=(Type=Bone,Name=\"deform_groin_r\"),Weight=0.1676),(Item=(Type=Bone,Name=\"spine_02\"),Weight=0.0719),(Item=(Type=Bone,Name=\"thigh_twist_01_r\"),Weight=0.0369),(Item=(Type=Bone,Name=\"spine_04\"),Weight=0.0051 ...(truncated)"
    Weight=1
   )
   > &RigUnit_ParentConstraint(
    AdvancedSettings=(InterpolationType=Average,RotationOrderForFilter=XZY)
    Child=(Type=Bone,Name="base_beltBags_01_R")
    Parents="((Item=(Type=Bone,Name=\"pelvis\"),Weight=0.7236),(Item=(Type=Bone,Name=\"deform_groin_r\"),Weight=0.168),(Item=(Type=Bone,Name=\"spine_02\"),Weight=0.0813),(Item=(Type=Bone,Name=\"thigh_twist_01_r\"),Weight=0.0187),(Item=(Type=Bone,Name=\"spine_04\"),Weight=0.0065) ...(truncated)"
    Weight=1
   )
   > &RigUnit_ParentConstraint(
    AdvancedSettings=(InterpolationType=Average,RotationOrderForFilter=XZY)
    Child=(Type=Bone,Name="base_beltBuckle_F")
    Parents="((Item=(Type=Bone,Name=\"pelvis\"),Weight=0.7236),(Item=(Type=Bone,Name=\"spine_02\"),Weight=0.1278),(Item=(Type=Bone,Name=\"deform_groin_l\"),Weight=0.0565),(Item=(Type=Bone,Name=\"deform_groin_r\"),Weight=0.0451),(Item=(Type=Bone,Name=\"spine_04\"),Weight=0.0175), ...(truncated)"
    Weight=1
   )
   > RigVMFunctionReturnNode()
  ]}
 New Function_ContainedGraph {nodes [
   n1 = RigVMFunctionEntryNode()
   RigVMFunction_ControlFlowBranch = &RigVMFunction_ControlFlowBranch(Condition=@n1.Debug, ExecuteContext=@n1.ExecuteContext)
   RigVMFunctionReturnNode(ExecuteContext=@RigVMFunction_ControlFlowBranch.Completed)
   RigUnit_GetTransform = &RigUnit_GetTransform(Item=(Type=Control,Name="thigh_r_ctrl"), bInitial=False)
   &RigUnit_SetTransform(
    Item=(Type=Bone,Name="thigh_r")
    Weight=1
    bInitial=False
    Value=@RigUnit_GetTransform.Transform
    ExecutePin=@RigVMFunction_ControlFlowBranch.True
   )
   > &RigUnit_SetTransform(
    Item=(Type=Bone,Name="thigh_l")
    Weight=1
    bInitial=False
    Value=&RigUnit_GetTransform(Item=(Type=Control,Name="thigh_l_ctrl"), bInitial=False).Transform
   )
   > RigUnit_SetTransform = &RigUnit_SetTransform(
    Item=(Type=Bone,Name="calf_l")
    Weight=1
    bInitial=False
    Value=&RigUnit_GetTransform(Item=(Type=Control,Name="calf_l_ctrl"), bInitial=False).Transform
   )
   > &RigUnit_SetTransform(
    Item=(Type=Bone,Name="clavicle_r")
    Weight=1
    bInitial=False
    Value=&RigUnit_GetTransform(Item=(Type=Control,Name="clavicle_r_ctrl"), bInitial=False).Transform
   )
   > &RigUnit_SetTransform(
    Item=(Type=Bone,Name="clavicle_l")
    Weight=1
    bInitial=False
    Value=&RigUnit_GetTransform(Item=(Type=Control,Name="clavicle_l_ctrl"), bInitial=False).Transform
   )
   > &RigUnit_SetTransform(
    Item=(Type=Bone,Name="upperarm_r")
    Weight=1
    bInitial=False
    Value=&RigUnit_GetTransform(Item=(Type=Control,Name="upperarm_r_ctrl"), bInitial=False).Transform
   )
   > &RigUnit_SetTransform(
    Item=(Type=Bone,Name="upperarm_l")
    Weight=1
    bInitial=False
    Value=&RigUnit_GetTransform(Item=(Type=Control,Name="upperarm_l_ctrl"), bInitial=False).Transform
   )
   > &RigUnit_SetTransform(
    Item=(Type=Bone,Name="spine_01")
    Weight=1
    bInitial=False
    Value=&RigUnit_GetTransform(Item=(Type=Control,Name="spine_01_ctrl"), bInitial=False).Transform
   )
  ]}
 New Function_ContainedGraph {nodes [
   RigVMFunctionEntryNode()
   > &RigUnit_ModifyTransforms(
    ItemToModify="((Item=(Type=Bone,Name=\"master_kneePad\"),Transform=(Rotation=(X=0.4982,Y=0,Z=0,W=0.8671),Translation=(X=0,Y=0,Z=0),Scale3D=(X=1,Y=1,Z=1))),(Item=(Type=Bone,Name=\"kneePad\"),Transform=(Rotation=(X=-0.0844,Y=0,Z=0,W=-0.9964),Translation=(X=0,Y=-1.1425,Z=-0.14 ...(truncated)"
    WeightMinimum=0
    Weight=RigVMDispatchNode(BoneName=root, Default=0, Name=deformRig_l_knee_alpha) {template="DISPATCH_RigDispatch_GetAnimAttribute(in Name,in BoneName,in Default,out Value,out Found)"}.Value
   )
   > RigVMFunctionReturnNode()
  ]}
 New Function_ContainedGraph {nodes [
   RigVMFunctionEntryNode()
   > &RigUnit_ParentConstraint(
    AdvancedSettings=(InterpolationType=Average,RotationOrderForFilter=XZY)
    Child=(Type=Bone,Name="base_shoulderPad_L_01")
    Filter=(TranslationFilter=(bX=True,bY=True,bZ=True),RotationFilter=(bX=True,bY=True,bZ=True),ScaleFilter=(bX=false,bY=false,bZ=false))
    Parents=((Item=(Type=Bone,Name="deform_upperarm_up_l"),Weight=0.5),(Item=(Type=Bone,Name="upperarm_twist_01_l"),Weight=1),(Item=(Type=Bone,Name="clavicle_l"),Weight=0.2))
    Weight=1
   )
   n1 = RigVMDispatchNode(BoneName=root, Default=0, Name=deformRig_l_upperarm_fr_alpha) {template="DISPATCH_RigDispatch_GetAnimAttribute(in Name,in BoneName,in Default,out Value,out Found)"}
   n2 = RigVMDispatchNode(BoneName=root, Default=0, Name=deformRig_l_upperarm_bk_alpha) {template="DISPATCH_RigDispatch_GetAnimAttribute(in Name,in BoneName,in Default,out Value,out Found)"}
   Multiply = &RigVMFunction_MathDoubleMul(B=20, A=@n1.Value)
   Add = &RigVMFunction_MathDoubleAdd(A=@Multiply.Result, B=&RigVMFunction_MathDoubleMul(B=-20, A=@n2.Value).Result)
   RigVMFunction_MathQuaternionFromEuler = &RigVMFunction_MathQuaternionFromEuler(
    Euler=(X=0,Y=0,Z=0)
    Euler.X=@Add.Result
    Euler.Z=&RigVMFunction_MathDoubleAdd(
     B=&RigVMFunction_MathDoubleMul(B=-15, A=@n2.Value).Result
     A=&RigVMFunction_MathDoubleMul(B=15, A=@n1.Value).Result
    ).Result
   )
   > &RigUnit_ModifyTransforms(
    ItemToModify=((Item=(Type=Bone,Name="dyn_shoulderPad_L_01"),Transform=(Rotation=(X=0,Y=0,Z=0,W=1),Translation=(X=0,Y=0,Z=0),Scale3D=(X=1,Y=1,Z=1))))
    WeightMinimum=0
    Weight=&RigVMFunction_MathDoubleAdd(A=@n1.Value, B=@n2.Value).Result
    ItemToModify.0.Transform.Rotation=@RigVMFunction_MathQuaternionFromEuler.Result
   )
   n3 = RigVMDispatchNode(BoneName=root, Default=0, Name=deformRig_l_upperarm_up_alpha) {template="DISPATCH_RigDispatch_GetAnimAttribute(in Name,in BoneName,in Default,out Value,out Found)"}
   > &RigUnit_ModifyTransforms(
    ItemToModify="((Item=(Type=Bone,Name=\"dyn_shoulderPad_L_01\"),Transform=(Rotation=(X=-0.0011,Y=0.0302,Z=0.061,W=0.9977),Translation=(X=0.5903,Y=-2,Z=0.9977),Scale3D=(X=1,Y=1,Z=1))),(Item=(Type=Bone,Name=\"dyn_shoulderPad_L_02\"),Transform=(Rotation=(X=0,Y=0,Z=0.0698,W=0.99 ...(truncated)"
    WeightMinimum=0
    Weight=&RigVMFunction_MathDoubleSub(B=@n1.Value, A=@n3.Value).Result
   )
   AlphaInterp = &RigVMFunction_AlphaInterp(
    InterpSpeedDecreasing=7
    InterpSpeedIncreasing=8
    Scale=1
    bClampResult=true
    Value=&RigVMFunction_MathDoubleMul(A=1, B=@n3.Value).Result
   )
   > &RigUnit_ModifyTransforms(
    ItemToModify=((Item=(Type=Bone,Name="dyn_shoulderPad_L_01"),Transform=(Rotation=(X=0,Y=0,Z=0,W=1),Translation=(X=0,Y=0,Z=0),Scale3D=(X=1,Y=1,Z=1))))
    WeightMinimum=0
    Weight=@AlphaInterp.Result
   )
   > &RigUnit_ParentConstraint(
    AdvancedSettings=(InterpolationType=Average,RotationOrderForFilter=XZY)
    Child=(Type=Bone,Name="base_shoulderPad_R_01")
    Filter=(TranslationFilter=(bX=True,bY=True,bZ=True),RotationFilter=(bX=True,bY=True,bZ=True),ScaleFilter=(bX=false,bY=false,bZ=false))
    Parents=((Item=(Type=Bone,Name="deform_upperarm_up_r"),Weight=0.5),(Item=(Type=Bone,Name="upperarm_twist_01_r"),Weight=1),(Item=(Type=Bone,Name="clavicle_r"),Weight=0.1))
    Weight=1
   )
   n4 = RigVMDispatchNode(BoneName=root, Default=0, Name=deformRig_r_upperarm_bk_alpha) {template="DISPATCH_RigDispatch_GetAnimAttribute(in Name,in BoneName,in Default,out Value,out Found)"}
   n5 = RigVMDispatchNode(BoneName=root, Default=0, Name=deformRig_r_upperarm_fr_alpha) {template="DISPATCH_RigDispatch_GetAnimAttribute(in Name,in BoneName,in Default,out Value,out Found)"}
   n6 = &RigVMFunction_MathDoubleMul(B=-35, A=@n4.Value)
   n7 = &RigVMFunction_MathDoubleMul(B=35, A=@n5.Value)
   n8 = &RigVMFunction_MathDoubleMul(B=-20, A=@n4.Value)
   n9 = &RigVMFunction_MathDoubleMul(B=20, A=@n5.Value)
   > &RigUnit_ModifyTransforms(
    ItemToModify=((Item=(Type=Bone,Name="dyn_shoulderPad_R_01"),Transform=(Rotation=(X=0.1722,Y=0.0227,Z=0.1285,W=0.9764),Translation=(X=0,Y=1,Z=1),Scale3D=(X=1,Y=1,Z=1))))
    WeightMinimum=0
    Weight=&RigVMFunction_MathDoubleAdd(B=@n4.Value, A=@n5.Value).Result
    ItemToModify.0.Transform.Rotation=&RigVMFunction_MathQuaternionFromEuler(
     Euler=(X=0,Y=0,Z=0)
     Euler.X=&RigVMFunction_MathDoubleAdd(B=@n6.Result, A=@n7.Result).Result
     Euler.Z=&RigVMFunction_MathDoubleAdd(B=@n8.Result, A=@n9.Result).Result
    ).Result
   )
   > &RigUnit_ModifyTransforms(
    ItemToModify=((Item=(Type=Bone,Name="dyn_shoulderPad_R_01"),Transform=(Rotation=(X=0.0004,Y=0.0302,Z=0.0116,W=0.9995),Translation=(X=-0.5903,Y=0.0321,Z=-0.9977),Scale3D=(X=1,Y=1,Z=1))))
    WeightMinimum=0
    Weight=&RigVMFunction_MathDoubleSub(
     A=RigVMDispatchNode(BoneName=root, Default=0, Name=deformRig_r_upperarm_up_alpha) {template="DISPATCH_RigDispatch_GetAnimAttribute(in Name,in BoneName,in Default,out Value,out Found)"}.Value
     B=@n5.Value
    ).Result
   )
   > n10 = &RigUnit_SphericalPoseReader(
    ActiveRegionScaleFactors=(PositiveWidth=2.4,NegativeWidth=3,PositiveHeight=3,NegativeHeight=3.5)
    ActiveRegionSize=0.1
    Debug=(bDrawDebug=false,bDraw2D=False,bDrawLocalAxes=False,DebugScale=25,DebugSegments=20,DebugThickness=0.25)
    DriverAxis=(X=1,Y=0,Z=0)
    DriverItem=(Type=Bone,Name="clavicle_l")
    FalloffRegionScaleFactors=(PositiveWidth=2,NegativeWidth=1,PositiveHeight=1,NegativeHeight=10)
    FalloffSize=0.4
    FlipHeightScaling=False
    FlipWidthScaling=False
    OptionalParentItem=(Type=Bone,Name="spine_01")
    RotationOffset=(X=80,Y=0,Z=0)
   )
   > &RigUnit_ModifyTransforms(
    ItemToModify="((Item=(Type=Bone,Name=\"dyn_shoulderPad_L_01\"),Transform=(Rotation=(X=-0.13,Y=-0.0114,Z=0.0864,W=0.9877),Translation=(X=0,Y=0,Z=0),Scale3D=(X=1,Y=1,Z=1))),(Item=(Type=Bone,Name=\"dyn_shoulderPad_L_02\"),Transform=(Rotation=(X=0,Y=0,Z=0,W=1),Translation=(X=0, ...(truncated)"
    WeightMinimum=0
    Weight=&RigVMFunction_MathDoubleMul(A=@n10.OutputParam, B=RigVMVariableNode() {variable=is not Emoting}.Value).Result
   )
   > RigVMFunctionReturnNode()
   &RigUnit_ModifyTransforms(
    ExecutePin=()
    ItemToModify=((Item=(Type=Bone,Name="dyn_shoulderPad_L_01"),Transform=(Rotation=(X=0.0004,Y=0.0302,Z=0.0116,W=0.9995),Translation=(X=0.5903,Y=-0.0321,Z=0.9977),Scale3D=(X=1,Y=1,Z=1))))
    Weight=0
    WeightMinimum=0
   )
   &RigUnit_ParentConstraint(
    AdvancedSettings=(InterpolationType=Average,RotationOrderForFilter=XZY)
    Child=(Type=Bone,Name="base_shoulderPad_L_01")
    ExecutePin=()
    Filter=(TranslationFilter=(bX=True,bY=True,bZ=True),RotationFilter=(bX=True,bY=True,bZ=True),ScaleFilter=(bX=false,bY=false,bZ=false))
    Parents=((Item=(Type=Bone,Name="deform_upperarm_up_l"),Weight=0.5),(Item=(Type=Bone,Name="upperarm_twist_01_l"),Weight=1),(Item=(Type=Bone,Name="clavicle_l"),Weight=0.2))
    Weight=1
   )
  ]}
 New Function_ContainedGraph {nodes [
   RigVMFunctionEntryNode()
   n1 = RigVMDispatchNode(BoneName=root, Default=0, Name=deformRig_l_thigh_up_alpha) {template="DISPATCH_RigDispatch_GetAnimAttribute(in Name,in BoneName,in Default,out Value,out Found)"}
   > &RigUnit_ModifyTransforms(
    ItemToModify="((Item=(Type=Bone,Name=\"beltBags_01_L\"),Transform=(Rotation=(X=0,Y=0.015,Z=0,W=-0.9999),Translation=(X=0.4878,Y=-1.3,Z=0),Scale3D=(X=1,Y=1,Z=1))),(Item=(Type=Bone,Name=\"beltBags_02_L\"),Transform=(Rotation=(X=0,Y=0,Z=0,W=1),Translation=(X=0,Y=-0.8,Z=0),Scal ...(truncated)"
    WeightMinimum=0
    Weight=@n1.Value
   )
   n2 = RigVMDispatchNode(BoneName=root, Default=0, Name=deformRig_r_thigh_up_alpha) {template="DISPATCH_RigDispatch_GetAnimAttribute(in Name,in BoneName,in Default,out Value,out Found)"}
   > &RigUnit_ModifyTransforms(
    ItemToModify="((Item=(Type=Bone,Name=\"beltBags_01_R\"),Transform=(Rotation=(X=-0.0014,Y=0.0275,Z=-0.0508,W=-0.9983),Translation=(X=0.2888,Y=-1.3512,Z=0),Scale3D=(X=1,Y=1,Z=1))),(Item=(Type=Bone,Name=\"beltBags_02_R\"),Transform=(Rotation=(X=0,Y=0,Z=0,W=-1),Translation=(X=0 ...(truncated)"
    WeightMinimum=0
    Weight=@n2.Value
   )
   Multiply = &RigVMFunction_MathDoubleMul(B=@n1.Value, A=RigVMVariableNode() {variable=Is Crouching}.Value)
   AlphaInterp = &RigVMFunction_AlphaInterp(InterpSpeedDecreasing=10, InterpSpeedIncreasing=20, Scale=1, bClampResult=False, Value=@Multiply.Result)
   > &RigUnit_ModifyTransforms(
    ItemToModify="((Item=(Type=Bone,Name=\"beltBags_01_L\"),Transform=(Rotation=(X=0,Y=0,Z=0,W=1),Translation=(X=0,Y=0,Z=0),Scale3D=(X=1,Y=1,Z=1))),(Item=(Type=Bone,Name=\"beltBags_02_L\"),Transform=(Rotation=(X=0,Y=0,Z=0,W=1),Translation=(X=0,Y=-1,Z=0.7),Scale3D=(X=1,Y=1,Z=1)) ...(truncated)"
    WeightMinimum=0
    Weight=@AlphaInterp.Result
   )
   Maximum = &RigVMFunction_MathDoubleMax(A=@n1.Value, B=@n2.Value)
   Add = &RigVMFunction_MathDoubleAdd(
    A=&RigVMFunction_MathDoubleMul(A=-12.5, B=@n1.Value).Result
    B=&RigVMFunction_MathDoubleMul(A=12.5, B=@n2.Value).Result
   )
   RigVMFunction_MathQuaternionFromEuler = &RigVMFunction_MathQuaternionFromEuler(Euler=(X=0,Y=0,Z=0), RotationOrder=XYZ, Euler.X=@Add.Result)
   > &RigUnit_ModifyTransforms(
    ItemToModify=((Item=(Type=Bone,Name="beltBuckle_F"),Transform=(Rotation=(X=0,Y=0,Z=0,W=1),Translation=(X=0,Y=-0.8,Z=0),Scale3D=(X=1,Y=1,Z=1))))
    WeightMinimum=0
    Weight=@Maximum.Result
    ItemToModify.0.Transform.Rotation=@RigVMFunction_MathQuaternionFromEuler.Result
   )
   > RigVMFunctionReturnNode()
  ]}
 RigVMFunctionLibrary {nodes [
   Belt Bags = RigVMCollapseNode(ExecuteContext=())
   Controls = RigVMCollapseNode(Debug=False, ExecuteContext=())
   Knee Pad = RigVMCollapseNode(ExecuteContext=())
   Shoulder Pads = RigVMCollapseNode(ExecuteContext=())
   Surface Constraints = RigVMCollapseNode(ExecuteContext=())
  ]}
 RigVMFunction_Sequence_1_ContainedGraph {nodes [
   n1 = RigVMFunctionEntryNode()
   RigVMFunction_Sequence = &RigVMFunction_Sequence(ExecuteContext=@n1.ExecuteContext)
   n2 = &RigVMFunction_Sequence(ExecuteContext=@RigVMFunction_Sequence.B)
   RigVMFunctionReturnNode(A=@RigVMFunction_Sequence.A, B=@n2.A, C=@n2.B)
  ]}
 RigVMModel {nodes [
   RigUnit_BeginExecution = &RigUnit_BeginExecution()
   RigVMFunction_Sequence = RigVMAggregateNode(ExecuteContext=@RigUnit_BeginExecution.ExecutePin)
   Controls = RigVMFunctionReferenceNode(Debug=false, ExecuteContext=@RigVMFunction_Sequence.A)
   FN_DeformRig_FNC = RigVMFunctionReferenceNode(
    DeformRig_ Curve="(deformRig_l_upperarm_up_alpha_72_#1=true,deformRig_r_upperarm_up_alpha_73_#2=true,deformRig_l_upperarm_fr_alpha_74_#3=true,deformRig_r_upperarm_fr_alpha_75_CB51B80C4 ...(truncated)"
    DeformRig_Bool="(upperarm_1_#4=True,elbow_3_#5=True,wrist_7_#6=True,groin_8_#7=True,glute_17_#8=True,knee_9_40C408324CA261 ...(truncated)"
    DeformRig_Debug="(upperarm_debug=False,wrist_debug=False,lat_debug=False,groinGlute_debug=False,upperarm_debug_18_#4=False,wrist_debug_20_#6=False,groinGlute_debug_27_#7=False,lat_deb ...(truncated)"
    DeformRig_Value="(upperarm_up_18_#9=(Translation=(X=0,Y=0,Z=8),Scale3D=(X=0.7,Y=1,Z=1)),upperarm_fwd_fr_19_#10=(Translation=(X=0,Y=10,Z=0),Scale3D=(X=0.5,Y=1,Z=1)),upperarm_fwd_bk_20_9973AAD84EE0A1E349BC9E85067D4A3 ...(truncated)"
    ExecuteContext=@RigVMFunction_Sequence.B
   )
   Surface Constraints = RigVMFunctionReferenceNode(ExecuteContext=@RigVMFunction_Sequence.C)
   > Knee Pad = RigVMFunctionReferenceNode()
   > Shoulder Pads = RigVMFunctionReferenceNode()
   > Belt Bags = RigVMFunctionReferenceNode()
  ]}
]
hierarchy [
 root {loc=<0 0 0>, rot=<0 0 0 1>}
 pelvis {parent=root, loc=<0 -0.5301 98.0571>, rot=<0.0127 0.707 -0.0127 -0.707>}
 spine_01 {parent=pelvis, loc=<6.1208 -0.9328 0>, rot=<0 0 0.034 -0.9994>}
 spine_02 {parent=spine_01, loc=<6.5997 -0.2095 0>, rot=<0 0 0.0238 -0.9997>}
 spine_03 {parent=spine_02, loc=<6.4914 -0.1237 0>, rot=<0 0 0.0149 -0.9999>}
 spine_04 {parent=spine_03, loc=<6.9598 0.0794 0>, rot=<0 0 -0.0132 -0.9999>}
 spine_05 {parent=spine_04, loc=<6.9987 0 0>, rot=<0 0 -0.0015 -1>}
 neck_01 {parent=spine_05, loc=<11.3138 0.1529 0>, rot=<0 0 -0.2003 -0.9797>}
 neck_02 {parent=neck_01, loc=<4.5627 0 0>, rot=<0 0 0.0504 -0.9987>}
 head {parent=neck_02, loc=<5.3914 0 0>, rot=<0 0 0.115 -0.9934>}
 clavicle_l {parent=spine_05, loc=<5.4831 -0.1053 -3.2946>, rot=<0.0465 -0.7289 -0.009 -0.6829>}
 upperarm_l {parent=clavicle_l, loc=<9.8658 -1.2136 -0.4105>, rot=<0.0164 -0.3245 -0.0269 -0.9454>}
 lowerarm_l {parent=upperarm_l, loc=<26.465 -0.8345 0.259>, rot=<0.0036 0.0161 -0.2298 -0.9731>}
 hand_l {parent=lowerarm_l, loc=<22.7355 -0.3203 -0.6105>, rot=<-0.6784 -0.0223 0.0735 -0.7307>}
 thumb_01_l {parent=hand_l, loc=<1.3118 -0.3191 -2.459>, rot=<0.6546 -0.4042 0.0664 -0.6354>}
 thumb_02_l {parent=thumb_01_l, loc=<5.9336 -0.0448 0.0401>, rot=<-0.0037 0.0213 0.298 -0.9543>}
 thumb_03_l {parent=thumb_02_l, loc=<2.2223 0.3682 -0.0907>, rot=<0.0118 -0.013 0.0066 -0.9998>}
 index_metacarpal_l {parent=hand_l, loc=<1.6535 0.0183 -2.96>, rot=<0.0158 -0.0124 -0.0173 -0.9996>}
 index_01_l {parent=index_metacarpal_l, loc=<8.3008 0.0357 -0.0524>, rot=<-0.0044 0.0485 0.0924 -0.9945>}
 index_02_l {parent=index_01_l, loc=<3.3587 -0.3004 0.0452>, rot=<-0.002 0.0027 0.3225 -0.9466>}
 index_03_l {parent=index_02_l, loc=<2.5578 -0.123 -0.0252>, rot=<-0.002 0.0067 0.2558 -0.9667>}
 middle_metacarpal_l {parent=hand_l, loc=<2.1856 -0.1725 -0.5984>, rot=<0.006 0.0168 -0.0329 -0.9993>}
 middle_01_l {parent=middle_metacarpal_l, loc=<7.3275 -0.2447 0.0129>, rot=<-0.0208 0.0042 0.0858 -0.9961>}
 middle_02_l {parent=middle_01_l, loc=<4.0123 0.1401 0.1175>, rot=<-0.0218 0.0258 0.3512 -0.9357>}
 middle_03_l {parent=middle_02_l, loc=<2.6077 0.0534 0.1373>, rot=<0.0082 0.0188 0.1708 -0.9851>}
 ring_metacarpal_l {parent=hand_l, loc=<2.2074 -0.3625 1.3638>, rot=<-0.0284 0.0722 -0.0458 -0.9959>}
 ring_01_l {parent=ring_metacarpal_l, loc=<7.0295 -0.6197 -0.1446>, rot=<-0.0489 -0.0252 0.1892 -0.9804>}
 ring_02_l {parent=ring_01_l, loc=<3.6174 0.4647 -0.1027>, rot=<0 0 0.2806 -0.9598>}
 ring_03_l {parent=ring_02_l, loc=<2.4152 0.114 0.0199>, rot=<0 0 0.1189 -0.9929>}
 pinky_metacarpal_l {parent=hand_l, loc=<1.7616 -0.9753 3.3814>, rot=<-0.0397 0.0814 -0.069 -0.9935>}
 pinky_01_l {parent=pinky_metacarpal_l, loc=<7.1709 -0.4831 -0.1094>, rot=<-0.0654 -0.0035 0.2242 -0.9723>}
 pinky_02_l {parent=pinky_01_l, loc=<2.8274 0.0893 -0.1043>, rot=<0.0037 0.0011 0.2854 -0.9584>}
 pinky_03_l {parent=pinky_02_l, loc=<2.2046 0.0307 -0.0078>, rot=<-0.0002 0.0098 0.1445 -0.9895>}
 deform_wrist_dn_base_l {parent=hand_l, loc=<0 -1.5326 0>, rot=<0 0 0 -1>}
 deform_wrist_dn_l {parent=deform_wrist_dn_base_l, loc=<0 0 0>, rot=<0 0 0 1>}
 deform_wrist_up_base_l {parent=hand_l, loc=<0 2.3009 0>, rot=<0 0 0 -1>}
 deform_wrist_up_l {parent=deform_wrist_up_base_l, loc=<0 0 0>, rot=<0 0 0 1>}
 lowerarm_twist_02_l {parent=lowerarm_l, loc=<10.6707 0 0>, rot=<0 0 0 1>}
 lowerarm_twist_01_l {parent=lowerarm_l, loc=<18.316 0 0>, rot=<0 0 0 1>}
 deform_elbow_in_base_l {parent=lowerarm_l, loc=<0.1834 3.6062 -0.2572>, rot=<0 0 0 -1>}
 deform_elbow_in_l {parent=deform_elbow_in_base_l, loc=<0 0 0>, rot=<0 0 0 1>}
 deform_elbow_out_base_l {parent=lowerarm_l, loc=<0 -4 0>, rot=<0 0 0 1>}
 deform_elbow_out_l {parent=deform_elbow_out_base_l, loc=<0 0 0>, rot=<0 0 0 1>}
 upperarm_twist_01_l {parent=upperarm_l, loc=<7.2135 -0.1134 0>, rot=<0 0 0 1>}
 deform_upperarm_bk_base_l {parent=upperarm_twist_01_l, loc=<-6.4053 -4.3746 0.314>, rot=<0 0 0 -1>}
 deform_upperarm_bk_l {parent=deform_upperarm_bk_base_l, loc=<0 0 0>, rot=<0 0 0 1>}
 deform_upperarm_fr_base_l {parent=upperarm_twist_01_l, loc=<-5.6312 5.092 0.3354>, rot=<0 0 0 -1>}
 deform_upperarm_fr_l {parent=deform_upperarm_fr_base_l, loc=<0 0 0>, rot=<0 0 0 1>}
 deform_upperarm_up_base_l {parent=upperarm_twist_01_l, loc=<-8.891 0.4825 3.9517>, rot=<0 0 0 -1>}
 deform_upperarm_up_l {parent=deform_upperarm_up_base_l, loc=<0 0 0>, rot=<0 0 0 1>}
 base_shoulderPad_L_01 {parent=upperarm_twist_01_l, loc=<-7.2135 0.1134 0>, rot=<0 0 0 1>}
 dyn_shoulderPad_L_01 {parent=base_shoulderPad_L_01, loc=<0 0 0>, rot=<0 0 0 1>}
 dyn_shoulderPad_L_02 {parent=dyn_shoulderPad_L_01, loc=<-5.9504 0.2786 0.8953>, rot=<-0.0124 0.3564 0.0324 -0.9337>}
 upperarm_twist_02_l {parent=upperarm_l, loc=<19.7998 -0.1134 0>, rot=<0 0 0 1>}
 clavicle_r {parent=spine_05, loc=<5.4831 -0.1053 3.2946>, rot=<0.7289 0.0465 -0.6829 0.009>}
 upperarm_r {parent=clavicle_r, loc=<-9.8658 1.2136 0.4105>, rot=<0.0164 -0.3245 -0.0269 -0.9454>}
 lowerarm_r {parent=upperarm_r, loc=<-26.465 0.8345 -0.259>, rot=<0.0036 0.0161 -0.2298 -0.9731>}
 hand_r {parent=lowerarm_r, loc=<-22.7355 0.3203 0.6106>, rot=<-0.6784 -0.0223 0.0735 -0.7307>}
 thumb_01_r {parent=hand_r, loc=<-1.3118 0.3191 2.459>, rot=<0.6546 -0.4042 0.0664 -0.6354>}
 thumb_02_r {parent=thumb_01_r, loc=<-5.9336 0.0448 -0.0401>, rot=<-0.0037 0.0213 0.298 -0.9543>}
 thumb_03_r {parent=thumb_02_r, loc=<-2.2223 -0.3682 0.0907>, rot=<0.0118 -0.013 0.0066 -0.9998>}
 index_metacarpal_r {parent=hand_r, loc=<-1.6535 -0.0183 2.96>, rot=<0.0158 -0.0124 -0.0173 -0.9996>}
 index_01_r {parent=index_metacarpal_r, loc=<-8.3008 -0.0357 0.0524>, rot=<-0.0044 0.0485 0.0924 -0.9945>}
 index_02_r {parent=index_01_r, loc=<-3.3587 0.3004 -0.0452>, rot=<-0.002 0.0027 0.3225 -0.9466>}
 index_03_r {parent=index_02_r, loc=<-2.5578 0.123 0.0252>, rot=<-0.002 0.0067 0.2558 -0.9667>}
 middle_metacarpal_r {parent=hand_r, loc=<-2.1856 0.1725 0.5984>, rot=<0.006 0.0168 -0.0329 -0.9993>}
 middle_01_r {parent=middle_metacarpal_r, loc=<-7.3275 0.2447 -0.0129>, rot=<-0.0208 0.0042 0.0858 -0.9961>}
 middle_02_r {parent=middle_01_r, loc=<-4.0123 -0.1401 -0.1175>, rot=<-0.0218 0.0258 0.3512 -0.9357>}
 middle_03_r {parent=middle_02_r, loc=<-2.6077 -0.0534 -0.1373>, rot=<0.0082 0.0188 0.1708 -0.9851>}
 ring_metacarpal_r {parent=hand_r, loc=<-2.2074 0.3625 -1.3638>, rot=<-0.0284 0.0722 -0.0458 -0.9959>}
 ring_01_r {parent=ring_metacarpal_r, loc=<-7.0295 0.6197 0.1446>, rot=<-0.0489 -0.0252 0.1892 -0.9804>}
 ring_02_r {parent=ring_01_r, loc=<-3.6174 -0.4647 0.1027>, rot=<0 0 0.2806 -0.9598>}
 ring_03_r {parent=ring_02_r, loc=<-2.4152 -0.114 -0.0199>, rot=<0 0 0.1189 -0.9929>}
 pinky_metacarpal_r {parent=hand_r, loc=<-1.7616 0.9753 -3.3814>, rot=<-0.0397 0.0814 -0.069 -0.9935>}
 pinky_01_r {parent=pinky_metacarpal_r, loc=<-7.1709 0.4831 0.1094>, rot=<-0.0654 -0.0035 0.2242 -0.9723>}
 pinky_02_r {parent=pinky_01_r, loc=<-2.8274 -0.0893 0.1043>, rot=<0.0037 0.0011 0.2854 -0.9584>}
 pinky_03_r {parent=pinky_02_r, loc=<-2.2046 -0.0307 0.0078>, rot=<-0.0002 0.0098 0.1445 -0.9895>}
 deform_wrist_dn_base_r {parent=hand_r, loc=<-0.0001 1.5328 0>, rot=<0 0 0 -1>}
 deform_wrist_dn_r {parent=deform_wrist_dn_base_r, loc=<0 0 0>, rot=<0 0 0 1>}
 deform_wrist_up_base_r {parent=hand_r, loc=<-0.0002 -2.3007 -0.0001>, rot=<0 0 0 -1>}
 deform_wrist_up_r {parent=deform_wrist_up_base_r, loc=<0 0 0>, rot=<0 0 0 1>}
 lowerarm_twist_02_r {parent=lowerarm_r, loc=<-10.6707 0 0>, rot=<0 0 0 -1>}
 lowerarm_twist_01_r {parent=lowerarm_r, loc=<-18.316 0 0>, rot=<0 0 0 -1>}
 deform_elbow_in_base_r {parent=lowerarm_r, loc=<-0.1834 -3.6062 0.2572>, rot=<0 0 0 -1>}
 deform_elbow_in_r {parent=deform_elbow_in_base_r, loc=<0 0 0>, rot=<0 0 0 1>}
 deform_elbow_out_base_r {parent=lowerarm_r, loc=<0 4 -0.0001>, rot=<0 0 0 -1>}
 deform_elbow_out_r {parent=deform_elbow_out_base_r, loc=<0 0 0>, rot=<0 0 0 1>}
 upperarm_twist_01_r {parent=upperarm_r, loc=<-7.2135 0.1134 0>, rot=<0 0 0 -1>}
 deform_upperarm_bk_base_r {parent=upperarm_twist_01_r, loc=<6.4053 4.3746 -0.314>, rot=<0 0 0 -1>}
 deform_upperarm_bk_r {parent=deform_upperarm_bk_base_r, loc=<0 0 0>, rot=<0 0 0 1>}
 deform_upperarm_fr_base_r {parent=upperarm_twist_01_r, loc=<5.6312 -5.092 -0.3354>, rot=<0 0 0 -1>}
 deform_upperarm_fr_r {parent=deform_upperarm_fr_base_r, loc=<0 0 0>, rot=<0 0 0 1>}
 deform_upperarm_up_base_r {parent=upperarm_twist_01_r, loc=<8.891 -0.4825 -3.9517>, rot=<0 0 0 -1>}
 deform_upperarm_up_r {parent=deform_upperarm_up_base_r, loc=<0 0 0>, rot=<0 0 0 1>}
 base_shoulderPad_R_01 {parent=upperarm_twist_01_r, loc=<7.2135 -0.1134 0>, rot=<0 0 0 -1>}
 dyn_shoulderPad_R_01 {parent=base_shoulderPad_R_01, loc=<0 0 0>, rot=<0 0 0 1>}
 dyn_shoulderPad_R_02 {parent=dyn_shoulderPad_R_01, loc=<5.9504 -0.2786 -0.8953>, rot=<-0.0124 0.3564 0.0324 -0.9337>}
 upperarm_twist_02_r {parent=upperarm_r, loc=<-19.7998 0.1134 0>, rot=<0 0 0 -1>}
 attach_backpack {parent=spine_05, loc=<-0.123 -13.473 0>, rot=<-0.0093 -0.707 0.0093 -0.707>}
 Backpack_BR {parent=attach_backpack, type=Null, loc=<0 0 0>, rot=<0 0 0 1>}
 attach_cape {parent=spine_05, loc=<9.017 -11.714 0>, rot=<0.0616 -0.7044 -0.0616 -0.7044>}
 cape {parent=attach_cape, type=Null, loc=<0 0 0>, rot=<0 0 0 1>}
 deform_pec_aim_base_l {parent=spine_04, loc=<3.2456 9.3309 -1.2237>, rot=<0.0603 -0.4033 0.3318 -0.8507>}
 deform_pec_aim_l {parent=deform_pec_aim_base_l, loc=<0 0 0>, rot=<0 0 0 1>}
 deform_pec_base_l {parent=deform_pec_aim_l, loc=<9.3665 0 0>, rot=<0 0 0 1>}
 deform_pec_l {parent=deform_pec_base_l, loc=<0 0 0>, rot=<0 0 0 1>}
 deform_pec_aim_base_r {parent=spine_04, loc=<3.2452 9.3309 1.2237>, rot=<-0.4033 -0.0603 0.8507 0.3318>}
 deform_pec_aim_r {parent=deform_pec_aim_base_r, loc=<0 0 0>, rot=<0 0 0 1>}
 deform_pec_base_r {parent=deform_pec_aim_r, loc=<-9.3667 -0.0003 -0.0004>, rot=<0 0 0 1>}
 deform_pec_r {parent=deform_pec_base_r, loc=<0 0 0>, rot=<0 0 0 1>}
 thigh_l {parent=pelvis, loc=<-1.0855 -0.2109 -8.7053>, rot=<0.0009 0.0355 0.0155 -0.9993>}
 calf_l {parent=thigh_l, loc=<-42.5284 -0.0019 -0.7297>, rot=<0.0006 -0.0161 -0.0391 -0.9991>}
 foot_l {parent=calf_l, loc=<-46.7401 -2.5014 -0.592>, rot=<0.0427 -0.0238 0.0353 -0.9982>}
 ball_l {parent=foot_l, loc=<-3.5658 10.8066 -0.1544>, rot=<0.0001 -0.0001 -0.7063 -0.7079>}
 calf_twist_02_l {parent=calf_l, loc=<-19.1546 -1.7943 -0.2638>, rot=<0.0237 0.0095 -0.0035 -0.9997>}
 calf_twist_01_l {parent=calf_l, loc=<-35.8483 -1.9197 -0.5794>, rot=<0.0237 0.0095 -0.0035 -0.9997>}
 deform_knee_in_base_l {parent=calf_l, loc=<0.8031 -5.0231 0.2456>, rot=<0 0 0 -1>}
 deform_knee_in_l {parent=deform_knee_in_base_l, loc=<0 0 0>, rot=<0 0 0 1>}
 deform_knee_out_base_l {parent=calf_l, loc=<0 4 0>, rot=<0 0 0 -1>}
 deform_knee_out_l {parent=deform_knee_out_base_l, loc=<0 0 0>, rot=<0 0 0 1>}
 master_kneePad {parent=calf_l, loc=<0 0 0>, rot=<-0.0309 -0.7201 0.0276 -0.6927>}
 kneePad {parent=master_kneePad, loc=<-0.4778 6.6306 0.4991>, rot=<0.1695 0.0485 0.0014 -0.9843>}
 thigh_twist_01_l {parent=thigh_l, loc=<-27.952 0 0>, rot=<0.0013 0 0 -1>}
 deform_groin_base_l {parent=thigh_l, loc=<0.4801 5.6776 -0.2486>, rot=<0 0 0 -1>}
 deform_groin_l {parent=deform_groin_base_l, loc=<0 0 0>, rot=<0 0 0 1>}
 thigh_r {parent=pelvis, loc=<-1.0855 -0.2109 8.7053>, rot=<0.0355 -0.0009 0.9993 0.0155>}
 calf_r {parent=thigh_r, loc=<42.5284 0.0019 0.7297>, rot=<0.0006 -0.0161 -0.0391 -0.9991>}
 foot_r {parent=calf_r, loc=<46.7401 2.5014 0.592>, rot=<0.0427 -0.0238 0.0353 -0.9982>}
 ball_r {parent=foot_r, loc=<3.5658 -10.8066 0.1544>, rot=<0.0001 -0.0001 -0.7063 -0.7079>}
 calf_twist_02_r {parent=calf_r, loc=<19.1546 1.7943 0.2638>, rot=<0.0237 0.0095 -0.0035 -0.9997>}
 masterShoeLace {parent=calf_twist_02_r, loc=<3.2806 -3.1817 0.8301>, rot=<-0.6847 0.0133 0.727 0.0496>}
 base_shoeLace_1_01 {parent=masterShoeLace, loc=<-0.4433 -0.1447 -0.2064>, rot=<-0.2819 -0.6046 0.6312 -0.3957>}
 dyn_shoeLace_1_01 {parent=base_shoeLace_1_01, loc=<0 0 0>, rot=<0 0 0 1>}
 dyn_shoeLace_1_02 {parent=dyn_shoeLace_1_01, loc=<4.6661 0 0>, rot=<-0.0042 0.0243 -0.1425 -0.9895>}
 base_shoeLace_2_01 {parent=masterShoeLace, loc=<0.5251 0.2236 -0.3375>, rot=<-0.5865 -0.3885 0.4772 -0.5267>}
 dyn_shoeLace_2_01 {parent=base_shoeLace_2_01, loc=<0 0 0>, rot=<0 0 0 1>}
 dyn_shoeLace_2_02 {parent=dyn_shoeLace_2_01, loc=<3.9488 0 0>, rot=<0.0343 0.029 0.0596 -0.9972>}
 calf_twist_01_r {parent=calf_r, loc=<35.8483 1.9197 0.5794>, rot=<0.0237 0.0095 -0.0035 -0.9997>}
 deform_knee_in_base_r {parent=calf_r, loc=<-0.8031 5.0231 -0.2457>, rot=<0 0 0 -1>}
 deform_knee_in_r {parent=deform_knee_in_base_r, loc=<0 0 0>, rot=<0 0 0 1>}
 deform_knee_out_base_r {parent=calf_r, loc=<0 -4 0>, rot=<0 0 0 -1>}
 deform_knee_out_r {parent=deform_knee_out_base_r, loc=<0 0 0>, rot=<0 0 0 1>}
 thigh_twist_01_r {parent=thigh_r, loc=<27.952 0 0>, rot=<0.0013 0 0 -1>}
 deform_groin_base_r {parent=thigh_r, loc=<-0.4801 -5.6776 0.2486>, rot=<0 0 0 -1>}
 deform_groin_r {parent=deform_groin_base_r, loc=<0 0 0>, rot=<0 0 0 1>}
 deform_glute_base_l {parent=pelvis, loc=<-7.3466 -9.0213 -9.1245>, rot=<0.0172 0.031 -0.4477 -0.8935>}
 deform_glute_l {parent=deform_glute_base_l, loc=<0 0 0>, rot=<0 0 0 1>}
 deform_glute_base_r {parent=pelvis, loc=<-7.3464 -9.0213 9.1245>, rot=<0.031 -0.0172 0.8935 -0.4477>}
 deform_glute_r {parent=deform_glute_base_r, loc=<0 0 0>, rot=<0 0 0 1>}
 base_beltBuckle_B {parent=pelvis, loc=<5.7353 -10.6066 -0.01>, rot=<0.8129 -0.5823 0.008 -0.0053>}
 beltBuckle_B {parent=base_beltBuckle_B, loc=<0 0 0>, rot=<0 0 0 1>}
 base_beltBags_04_R {parent=pelvis, loc=<6.6263 -6.6814 11.8109>, rot=<0.7188 -0.4999 0.2693 0.401>}
 beltBags_04_R {parent=base_beltBags_04_R, loc=<0 0 0>, rot=<0 0 0 1>}
 base_beltBags_03_R {parent=pelvis, loc=<6.5413 -1.6536 13.1983>, rot=<-0.6018 0.4528 -0.3934 -0.5273>}
 beltBags_03_R {parent=base_beltBags_03_R, loc=<0 0 0>, rot=<0 0 0 1>}
 base_beltBags_02_R {parent=pelvis, loc=<6.4779 3.8317 12.1144>, rot=<-0.4329 0.3033 -0.5526 -0.6444>}
 beltBags_02_R {parent=base_beltBags_02_R, loc=<0 0 0>, rot=<0 0 0 1>}
 base_beltBags_01_R {parent=pelvis, loc=<5.7746 7.883 7.6519>, rot=<-0.2368 0.1354 -0.6442 -0.7145>}
 beltBags_01_R {parent=base_beltBags_01_R, loc=<0 0 0>, rot=<0 0 0 1>}
 base_beltBuckle_F {parent=pelvis, loc=<5.5001 10.0871 0>, rot=<0 0 -0.6864 -0.7272>}
 beltBuckle_F {parent=base_beltBuckle_F, loc=<0 0 0>, rot=<0 0 0 1>}
 base_beltBags_01_L {parent=pelvis, loc=<5.7746 7.883 -7.652>, rot=<0.2318 -0.1399 -0.6433 -0.7162>}
 beltBags_01_L {parent=base_beltBags_01_L, loc=<0 0 0>, rot=<0 0 0 1>}
 base_beltBags_02_L {parent=pelvis, loc=<6.4779 3.8317 -12.1144>, rot=<0.4276 -0.3008 -0.5545 -0.6475>}
 beltBags_02_L {parent=base_beltBags_02_L, loc=<0 0 0>, rot=<0 0 0 1>}
 base_beltBags_03_L {parent=pelvis, loc=<6.5413 -1.6536 -13.1984>, rot=<0.6045 -0.4273 -0.3908 -0.5471>}
 beltBags_03_L {parent=base_beltBags_03_L, loc=<0 0 0>, rot=<0 0 0 1>}
 base_beltBags_04_L {parent=pelvis, loc=<6.6263 -6.6814 -11.8109>, rot=<0.7113 -0.5071 -0.2834 -0.3958>}
 beltBags_04_L {parent=base_beltBags_04_L, loc=<0 0 0>, rot=<0 0 0 1>}
 base_beltBags_05_L {parent=pelvis, loc=<6.2833 -9.9445 -8.1064>, rot=<0.7925 -0.5561 -0.0965 -0.2312>}
 beltBags_05_L {parent=base_beltBags_05_L, loc=<0 0 0>, rot=<0 0 0 1>}
 base_beltBags_05_R {parent=pelvis, loc=<6.2833 -9.9445 8.1064>, rot=<0.8014 -0.5444 0.0962 0.2282>}
 beltBags_05_R {parent=base_beltBags_05_R, loc=<0 0 0>, rot=<0 0 0 1>}
 clavicle_r_ctrl {controlType=EULER_TRANSFORM, type=Control, loc=<0 0 0>, rot=<0 0 0 1>}
 upperarm_r_ctrl {controlType=EULER_TRANSFORM, parent=clavicle_r_ctrl, type=Control, loc=<0 0 0>, rot=<0 0 0 1>}
 clavicle_l_ctrl {controlType=EULER_TRANSFORM, type=Control, loc=<0 0 0>, rot=<0 0 0 1>}
 upperarm_l_ctrl {controlType=EULER_TRANSFORM, parent=clavicle_l_ctrl, type=Control, loc=<0 0 0>, rot=<0 0 0 1>}
 spine_01_ctrl {controlType=EULER_TRANSFORM, type=Control, loc=<0 0 0>, rot=<0 0 0 1>}
 thigh_l_ctrl {controlType=EULER_TRANSFORM, parent=spine_01_ctrl, type=Control, loc=<0 0 0>, rot=<0 0 0 1>}
 calf_l_ctrl {controlType=EULER_TRANSFORM, parent=thigh_l_ctrl, type=Control, loc=<0 0 0>, rot=<0 0 0 1>}
 thigh_r_ctrl {controlType=EULER_TRANSFORM, parent=spine_01_ctrl, type=Control, loc=<0 0 0>, rot=<0 0 0 1>}
]
variables [
 Is Crouching {type=double, default=0}
 is not Emoting {type=double, default=0}
]
