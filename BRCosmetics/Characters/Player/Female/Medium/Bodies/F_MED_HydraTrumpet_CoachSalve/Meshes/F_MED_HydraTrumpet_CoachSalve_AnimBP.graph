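# Post-process style Anim Blueprint for the F_MED_HydraTrumpet_CoachSalve player body, layering cosmetic bone fixes and secondary motion onto the parent mesh's pose.
# Copied pose feeds a Control Rig driven by crouch and emote alphas, then cached stages: shoelace constraint, shoulder-pad and waist-box offsets, AnimDynamics on shoelaces.
# A left-pinky rotation finishes the output.
# The event graph also sets zipline dynamics alpha and front-end standing state, neither read by the shown AnimGraph.
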AnimBlueprint /BRCosmetics/Characters/Player/Female/Medium/Bodies/F_MED_HydraTrumpet_CoachSalve/Meshes/F_MED_HydraTrumpet_CoachSalve_AnimBP
target_skeleton=&/F_MED_HydraTrumpet_CoachSalve_Skeleton
graphs [
 AnimGraph {nodes [
   n1 = "Transform (Modify) Bone - Bone: pinky_03_l"!(
    ComponentPose=Local To Component(LocalPose=Use cached pose 'Physics Setup'())
    bAlphaBoolEnabled:bool=True
    Translation="0,0,0"
    Rotation="-20,0,0"
    Scale="1,1,1"
   ) -> (Pose) {
    alphaInputType=Bool
    rotationMode=BMM_Additive
    rotationSpace=BCS_BoneSpace
    translationMode=BMM_Additive
    translationSpace=BCS_BoneSpace
    alphaBoolBlend {blendOption=Linear}
    boneToModify {boneName=pinky_03_l}
    rotation {pitch=-20}
   }
   Output Pose(Result=Component To Local(ComponentPose=@n1)) {layerGroup=None}
   n2 = Control Rig!(
    Source=Copy Pose From Mesh() {bUseAttachedParent=true}
    Alpha
    Is Crouching:double=Get is Crouching Alpha() -> (is Crouching Alpha:double)
    is not Emoting:double=Get is Not Emoting() -> (is Not Emoting:double)
   ) -> (Pose) {
    alphaBoolBlend {blendOption=Linear}
    controlRigAssetReference {blueprintRigClass=&/F_MED_HydraTrumpet_CoachSalve_CtrlRig.F_MED_HydraTrumpet_CoachSalve_CtrlRig_C}
   }
   Save cached pose 'CR Pose'(Pose=@n2) {cacheName=CR Pose}
   n3 = "Constraint - masterShoeLace (PPP)"!(
    ComponentPose=Local To Component(LocalPose=Use cached pose 'CR Pose'())
    Alpha=1
    ConstraintWeights_0=0.8272
    ConstraintWeights_1:float=0.1665
    ConstraintWeights_2:float=0.0062
   ) -> (Pose) {
    alphaBoolBlend {blendOption=Linear}
    boneToModify {boneName=masterShoeLace}
    constraintSetup [{targetBone {boneName=calf_twist_02_r}}, {targetBone {boneName=calf_r}}, {targetBone {boneName=calf_twist_01_r}}]
    constraintWeights [1,1,1]
   }
   Save cached pose 'Surface Constraints Setup'(Pose=Component To Local(ComponentPose=@n3)) {cacheName=Surface Constraints Setup}
   n4 = Anim Dynamics - Bone: dyn_shoeLace_1_01(ComponentPose=Local To Component(LocalPose=Use cached pose 'Pre Dynamics Correctives'()), Alpha=1, ExternalForce="0,0,0") {
    angularDampingOverride=1
    angularSpringConstant=20000
    bAngularSpring=true
    bOverrideAngularDamping=true
    bUseGravityOverride=true
    gravityScale=0.35
    simulationSpace=RootRelative
    alphaBoolBlend {blendOption=Linear}
    boundBone {boneName=dyn_shoeLace_1_01}
    componentLinearAccScale=<1 1 1>
    componentLinearVelScale=<1 1 1>
    gravityOverride {z=-980}
    physicsBodyDefinitions [{
      boundBone {boneName=dyn_shoeLace_1_01}
      boxExtents=<20 30 30>
      constraintSetup {angularLimitsMax=<15 15 15>, angularLimitsMin=<-15 -15 -15>, angularTarget=<1 0 0>}
      localJointOffset=<4 0 0>
     }]
   }
   n5 = Anim Dynamics - Bone: dyn_shoeLace_1_02(ComponentPose=@n4, Alpha=1, ExternalForce="0,0,0") {
    angularDampingOverride=1
    angularSpringConstant=20000
    bAngularSpring=true
    bOverrideAngularDamping=true
    bUseGravityOverride=true
    gravityScale=0.35
    simulationSpace=RootRelative
    alphaBoolBlend {blendOption=Linear}
    boundBone {boneName=dyn_shoeLace_1_02}
    componentLinearAccScale=<1 1 1>
    componentLinearVelScale=<1 1 1>
    gravityOverride {z=-980}
    physicsBodyDefinitions [{
      boundBone {boneName=dyn_shoeLace_1_02}
      boxExtents=<20 30 30>
      constraintSetup {angularLimitsMax=<15 15 15>, angularLimitsMin=<-15 -15 -15>, angularTarget=<1 0 0>}
      localJointOffset=<4 0 0>
     }]
   }
   n6 = Anim Dynamics - Bone: dyn_shoeLace_2_01(ComponentPose=@n5, Alpha=1, ExternalForce="0,0,0") {
    angularDampingOverride=1
    angularSpringConstant=20000
    bAngularSpring=true
    bOverrideAngularDamping=true
    bUseGravityOverride=true
    gravityScale=0.35
    simulationSpace=RootRelative
    alphaBoolBlend {blendOption=Linear}
    boundBone {boneName=dyn_shoeLace_2_01}
    componentLinearAccScale=<1 1 1>
    componentLinearVelScale=<1 1 1>
    gravityOverride {z=-980}
    physicsBodyDefinitions [{
      boundBone {boneName=dyn_shoeLace_2_01}
      boxExtents=<20 30 30>
      constraintSetup {angularLimitsMax=<15 15 15>, angularLimitsMin=<-15 -15 -15>, angularTarget=<1 0 0>}
      localJointOffset=<4 0 0>
     }]
   }
   n7 = Anim Dynamics - Bone: dyn_shoeLace_2_02(ComponentPose=@n6, Alpha=1, ExternalForce="0,0,0") {
    angularDampingOverride=1
    angularSpringConstant=20000
    bAngularSpring=true
    bOverrideAngularDamping=true
    bUseGravityOverride=true
    gravityScale=0.35
    simulationSpace=RootRelative
    alphaBoolBlend {blendOption=Linear}
    boundBone {boneName=dyn_shoeLace_2_02}
    componentLinearAccScale=<1 1 1>
    componentLinearVelScale=<1 1 1>
    gravityOverride {z=-980}
    physicsBodyDefinitions [{
      boundBone {boneName=dyn_shoeLace_2_02}
      boxExtents=<20 30 30>
      constraintSetup {angularLimitsMax=<15 15 15>, angularLimitsMin=<-15 -15 -15>, angularTarget=<1 0 0>}
      localJointOffset=<4 0 0>
     }]
   }
   Save cached pose 'Physics Setup'(Pose=Component To Local(ComponentPose=@n7)) {cacheName=Physics Setup}
   n8 = "Transform (Modify) Bone - Bone: dyn_shoulderPad_L_01"(
    ComponentPose=Local To Component(LocalPose=Use cached pose 'Surface Constraints Setup'())
    Alpha=1
    Translation="0,1,0"
    Rotation="0,0,-4"
    Scale="1,1,1"
   ) {
    rotationMode=BMM_Additive
    rotationSpace=BCS_BoneSpace
    translationMode=BMM_Additive
    translationSpace=BCS_BoneSpace
    alphaBoolBlend {blendOption=Linear}
    alphaScaleBiasClamp {bInterpResult=true}
    boneToModify {boneName=dyn_shoulderPad_L_01}
    rotation {roll=-4}
    translation {y=1}
   }
   n9 = "Transform (Modify) Bone - Bone: dyn_shoulderPad_R_01"(ComponentPose=@n8, Alpha=1, Translation="0,-0.5,0", Rotation="0,0,0", Scale="1,1,1") {
    rotationMode=BMM_Additive
    rotationSpace=BCS_BoneSpace
    translationMode=BMM_Additive
    translationSpace=BCS_BoneSpace
    alphaBoolBlend {blendOption=Linear}
    alphaScaleBiasClamp {bInterpResult=true}
    boneToModify {boneName=dyn_shoulderPad_R_01}
    translation {y=-0.5}
   }
   n10 = "Transform (Modify) Bone - Bone: dyn_shoulderPad_R_01"(ComponentPose=@n9, Alpha=1, Translation="0,0,0", Rotation="0,0,0", Scale="1,1,1") {
    rotationMode=BMM_Additive
    rotationSpace=BCS_BoneSpace
    translationMode=BMM_Additive
    translationSpace=BCS_BoneSpace
    alphaBoolBlend {blendOption=Linear}
    alphaScaleBiasClamp {bInterpResult=true}
    boneToModify {boneName=dyn_shoulderPad_R_01}
   }
   n11 = "Transform (Modify) Bone - Bone: dyn_shoulderPad_L_01"(ComponentPose=@n10, Alpha=1, Translation="0,1,0", Rotation="0,-7,0", Scale="1,1,1") {
    rotationMode=BMM_Additive
    rotationSpace=BCS_BoneSpace
    translationMode=BMM_Additive
    translationSpace=BCS_BoneSpace
    alphaBoolBlend {blendOption=Linear}
    alphaScaleBiasClamp {bInterpResult=true}
    boneToModify {boneName=dyn_shoulderPad_L_01}
    rotation {yaw=-7}
    translation {y=1}
   }
   n12 = "Transform (Modify) Bone - Bone: dyn_waistBox"(ComponentPose=@n11, Alpha=1, Translation="0,0,8", Rotation="0,0,0", Scale="1,1,1") {
    rotationMode=BMM_Additive
    rotationSpace=BCS_BoneSpace
    translationMode=BMM_Additive
    translationSpace=BCS_BoneSpace
    alphaBoolBlend {blendOption=Linear}
    alphaScaleBiasClamp {bInterpResult=true}
    boneToModify {boneName=dyn_waistBox}
    translation {z=8}
   }
   n13 = "Transform (Modify) Bone - Bone: dyn_shoulderPad_L_01"(ComponentPose=@n12, Alpha=1, Translation="0,0,0", Rotation="0,-5.5,0", Scale="1,1,1") {
    rotationMode=BMM_Additive
    rotationSpace=BCS_BoneSpace
    translationMode=BMM_Additive
    translationSpace=BCS_BoneSpace
    alphaBoolBlend {blendOption=Linear}
    alphaScaleBiasClamp {bInterpResult=true}
    boneToModify {boneName=dyn_shoulderPad_L_01}
    rotation {yaw=-5.5}
   }
   n14 = "Transform (Modify) Bone - Bone: dyn_shoulderPad_R_01"(ComponentPose=@n13, Alpha=1, Translation="0,0,0", Rotation="0,-3,0", Scale="1,1,1") {
    rotationMode=BMM_Additive
    rotationSpace=BCS_BoneSpace
    translationMode=BMM_Additive
    translationSpace=BCS_BoneSpace
    alphaBoolBlend {blendOption=Linear}
    alphaScaleBiasClamp {bInterpResult=true}
    boneToModify {boneName=dyn_shoulderPad_R_01}
    rotation {yaw=-3}
   }
   Save cached pose 'Pre Dynamics Correctives'(Pose=Component To Local(ComponentPose=@n14)) {cacheName=Pre Dynamics Correctives}
  ]}
 EventGraph {nodes [
   Gravity Override State Function DEPRECATED(CharacterPartAnimInstance=Self-Reference(), GameTimeInSeconds=GetGameTimeInSeconds(), GravityOverrideParams=Get Gravity Override Params())
   > Set GravityOverride(GravityOverride="0, 0, 0")
   Event BlueprintUpdateAnimation()
   n1 = NOT Boolean(A=Get bIsMoving2D())
   n2 = Get bIsCrouching()
   > Set is Crouching Alpha(is Crouching Alpha:double="To Float (Boolean)"(InBool=AND Boolean(A=@n2, B=@n1))) -> (Output_Get:double)
   n3 = Get bIsPlayingEmote()
   > Set is Not Emoting(is Not Emoting:double="To Float (Boolean)"(InBool=NOT Boolean(A=@n3))) -> (Output_Get:double)
   > n4 = Sequence()
   > n5 = Branch(Condition=Get OnZipline() -> (OnZipline:bool))
   > Set animDynamics Alpha(animDynamics Alpha:double=0) -> (Output_Get:double)
   Set animDynamics Alpha(execute=@n5.else, animDynamics Alpha=1)
   n6 = OR Boolean!(
    A=Get bIsSkyDiving()
    B=Get bIsPlayingEmote()
    C:bool=Get bIsCrouching()
    D:bool=Get bIsTargeting()
    E:bool=Get bIsPlayingMeleeAnim()
    F:bool=Get bIsSliding()
    G:bool=Get bIsMoving2D()
   ) -> (ReturnValue)
   Set StandingFrontend(execute=@n4.then_1, StandingFrontend:bool=AND Boolean(A=Get bIsInFrontEnd(), B=NOT Boolean(A=@n6))) -> (Output_Get:bool)
   n7 = Event BlueprintProcessAnimInputEvent()
   > n8 = Branch(Condition=IsNamed(self=@n7.AnimInputEvent, InEventName=aie.Player.Zipline.Entered))
   > Set OnZipline(OnZipline:bool=1) -> (Output_Get:bool)
   Branch(execute=@n8.else, Condition=IsNamed(self=@n7.AnimInputEvent, InEventName=aie.Player.Zipline.Exited))
   > Set OnZipline(OnZipline=false)
  ]}
]
variables [
 animDynamics Alpha {type=number}
 Gravity Override Params {type=GravityOverrideParamsStruct}
 GravityOverride {type=Vector}
 is Crouching Alpha {type=number}
 is Not Emoting {type=number}
 is Parachute Open {type=boolean}
 OnZipline {type=boolean}
 StandingFrontend {type=boolean}
]
