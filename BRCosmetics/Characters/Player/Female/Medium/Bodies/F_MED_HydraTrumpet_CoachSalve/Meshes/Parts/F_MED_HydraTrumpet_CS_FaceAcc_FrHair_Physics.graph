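# Physics asset for the CoachSalve female character's hair (FaceAcc FrHair part), with capsule bodies for side and back strands and the upper body.
# Four side strands run as dyn chains from small base anchors on the head, with limited swing and mostly locked twist.
# Head, spine, clavicle and arm capsules form body volumes linked by angularly free constraints.
# Back strand bodies have shapes but no constraints here.
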PhysicsAsset /BRCosmetics/Characters/Player/Female/Medium/Bodies/F_MED_HydraTrumpet_CoachSalve/Meshes/Parts/F_MED_HydraTrumpet_CS_FaceAcc_FrHair_Physics
bodies {
 base_backHairStrand_1_02 [{shapeType=Capsule, length=0.505, radius=0.505, shapeName=base_backHairStrand_1_02_capsule, rotation=<rot 0 -90 0>}]
 base_backHairStrand_2_03 [{shapeType=Capsule, length=0.505, radius=0.505, shapeName=base_backHairStrand_2_03_capsule, rotation=<rot 0 -90 0>}]
 base_backHairStrand_3_01 [{shapeType=Capsule, length=0.505, radius=0.505, shapeName=base_backHairStrand_3_01_capsule, rotation=<rot 0 -90 0>}]
 base_backHairStrand_3_03 [{shapeType=Capsule, length=0.505, radius=0.505, shapeName=base_backHairStrand_3_03_capsule, rotation=<rot 0 -90 0>}]
 base_backHairStrand_4_01 [{shapeType=Capsule, length=0.505, radius=0.505, shapeName=base_backHairStrand_4_01_capsule, rotation=<rot 0 -90 0>}]
 base_backHairStrand_5_03 [{shapeType=Capsule, length=0.505, radius=0.505, shapeName=base_backHairStrand_5_03_capsule, rotation=<rot 0 -90 0>}]
 base_backHairStrand_6_03 [{shapeType=Capsule, length=0.505, radius=0.505, shapeName=base_backHairStrand_6_03_capsule, rotation=<rot 0 -90 0>}]
 base_sideHairStrand_1_03_L [{shapeType=Capsule, length=0.505, radius=0.505, shapeName=base_sideHairStrand_1_03_L_capsule, rotation=<rot 0 -90 0>}]
 base_sideHairStrand_1_03_R [{shapeType=Capsule, length=0.505, radius=0.505, shapeName=base_sideHairStrand_1_03_R_capsule, rotation=<rot 0 -90 0>}]
 base_sideHairStrand_2_03_L [{shapeType=Capsule, length=0.505, radius=0.505, shapeName=base_sideHairStrand_2_03_L_capsule, rotation=<rot 0 -90 0>}]
 base_sideHairStrand_2_03_R [{shapeType=Capsule, length=0.505, radius=0.505, shapeName=base_sideHairStrand_2_03_R_capsule, rotation=<rot 0 -90 0>}]
 clavicle_l [{shapeType=Capsule, length=3, radius=2, shapeName=clavicle_l_capsule, rotation=<rot -90 0 0>}]
 clavicle_r [{shapeType=Capsule, length=3, radius=2, shapeName=clavicle_r_capsule, rotation=<rot -90 0 0>}]
 dyn_backHairStrand_1_02 [{
   shapeType=Capsule
   length=10.2216
   radius=2
   shapeName=dyn_backHairStrand_1_02_capsule
   center=<6.5277 2.6256 0>
   rotation=<rot 90 180 145.6222>
  }]
 dyn_backHairStrand_1_03 [{shapeType=Capsule, length=8.2166, radius=2, shapeName=dyn_backHairStrand_1_03_capsule, center=<6.3549 -0.2526 0>, rotation=<rot 90 0 0>}]
 dyn_backHairStrand_1_04 [{shapeType=Capsule, length=8.6539, radius=2, shapeName=dyn_backHairStrand_1_04_capsule, center=<6.6291 0.196 0>, rotation=<rot 90 0 0>}]
 dyn_backHairStrand_1_05 [{
   shapeType=Capsule
   length=8.1253
   radius=2
   shapeName=dyn_backHairStrand_1_05_capsule
   center=<4.8414 0.3083 0>
   rotation=<rot 90 178.2101 173.0984>
  }]
 dyn_backHairStrand_1_06 [{
   shapeType=Capsule
   length=8.2943
   radius=2
   shapeName=dyn_backHairStrand_1_06_capsule
   center=<3.901 0.2065 0>
   rotation=<rot 90 180 176.2761>
  }]
 dyn_backHairStrand_2_03 [{shapeType=Capsule, length=10.1304, radius=2, shapeName=dyn_backHairStrand_2_03_capsule, center=<6.0938 -0.1394 0>, rotation=<rot 90 0 0>}]
 dyn_backHairStrand_2_04 [{shapeType=Capsule, length=6.9784, radius=2, shapeName=dyn_backHairStrand_2_04_capsule, center=<5.372 0 -0.0988>, rotation=<rot 90 0 0>}]
 dyn_backHairStrand_2_05 [{
   shapeType=Capsule
   length=6.8597
   radius=1.74
   shapeName=dyn_backHairStrand_2_05_capsule
   center=<5.0882 -0.0437 0.2298>
   rotation=<rot 90 180 -178.2755>
  }]
 dyn_backHairStrand_2_06 [{
   shapeType=Capsule
   length=9.1428
   radius=1.6765
   shapeName=dyn_backHairStrand_2_06_capsule
   center=<5.6128 -0.2034 -0.1548>
   rotation=<rot 90 0 0>
  }]
 dyn_backHairStrand_3_03 [{
   shapeType=Capsule
   length=10.209
   radius=1.8962
   shapeName=dyn_backHairStrand_3_03_capsule
   center=<6.7976 -0.3312 -0.1938>
   rotation=<rot 90 0 0>
  }]
 dyn_backHairStrand_3_04 [{
   shapeType=Capsule
   length=7.2711
   radius=1.7455
   shapeName=dyn_backHairStrand_3_04_capsule
   center=<6.0512 -0.0897 -0.1094>
   rotation=<rot 90 0 0>
  }]
 dyn_backHairStrand_3_05 [{shapeType=Capsule, length=6.2912, radius=1.8017, shapeName=dyn_backHairStrand_3_05_capsule, center=<5.2333 0.0975 0>, rotation=<rot 90 0 0>}]
 dyn_backHairStrand_3_06 [{shapeType=Capsule, length=6.7101, radius=2, shapeName=dyn_backHairStrand_3_06_capsule, center=<4.8651 -0.0392 0>, rotation=<rot 90 0 0>}]
 dyn_backHairStrand_3_07 [{shapeType=Capsule, length=5.1927, radius=2, shapeName=dyn_backHairStrand_3_07_capsule, center=<3.9878 0 0>, rotation=<rot 90 0 0>}]
 dyn_backHairStrand_4_01 [{shapeType=Capsule, length=1, radius=2, shapeName=dyn_backHairStrand_3_01_capsule, center=<2 0 0>, rotation=<rot 90 0 0>}]
 dyn_backHairStrand_4_02 [{
   shapeType=Capsule
   length=7.5746
   radius=2
   shapeName=dyn_backHairStrand_3_01_capsule
   center=<7.0304 2.2922 -0.1926>
   rotation=<rot 90 153.4349 118.3243>
  }]
 dyn_backHairStrand_4_03 [{shapeType=Capsule, length=7.2104, radius=2, shapeName=dyn_backHairStrand_3_01_capsule, center=<4.7048 0 0.0283>, rotation=<rot 90 0 0>}]
 dyn_backHairStrand_4_04 [{shapeType=Capsule, length=7.0605, radius=1.4641, shapeName=dyn_backHairStrand_3_01_capsule, center=<4.6778 0 0.2152>, rotation=<rot 90 0 0>}]
 dyn_backHairStrand_5_03 [{
   shapeType=Capsule
   length=10.1626
   radius=2
   shapeName=dyn_backHairStrand_3_01_capsule
   center=<7.6417 -0.3062 -0.1139>
   rotation=<rot 90 0 0>
  }]
 dyn_backHairStrand_5_04 [{shapeType=Capsule, length=7.1905, radius=1.8197, shapeName=dyn_backHairStrand_3_01_capsule, center=<5.8876 0 0>, rotation=<rot 90 0 0>}]
 dyn_backHairStrand_5_05 [{shapeType=Capsule, length=7.8997, radius=2, shapeName=dyn_backHairStrand_3_01_capsule, center=<6.276 -0.3832 0>, rotation=<rot 90 0 0>}]
 dyn_backHairStrand_6_03 [{shapeType=Capsule, length=9.2821, radius=2, shapeName=dyn_backHairStrand_3_01_capsule, center=<7.1496 0 -0.2894>, rotation=<rot 90 0 0>}]
 dyn_backHairStrand_6_04 [{shapeType=Capsule, length=8.0002, radius=2, shapeName=dyn_backHairStrand_3_01_capsule, center=<6.5021 0.2346 0>, rotation=<rot 90 0 0>}]
 dyn_backHairStrand_6_05 [{
   shapeType=Capsule
   length=9.4559
   radius=2
   shapeName=dyn_backHairStrand_3_01_capsule
   center=<8.0842 -0.5029 -0.4004>
   rotation=<rot 90 0 0>
  }]
 dyn_sideHairStrand_1_03_L [{
   shapeType=Capsule
   length=11.1471
   radius=2
   shapeName=dyn_backHairStrand_3_01_capsule
   center=<9.8432 2.1269 0.6609>
   rotation=<rot 90 0 -18.5351>
  }]
 dyn_sideHairStrand_1_03_R [{
   shapeType=Capsule
   length=8.2995
   radius=1.8001
   shapeName=dyn_backHairStrand_3_01_capsule
   center=<10.5359 1.9279 -0.2988>
   rotation=<rot 90 180 156.7855>
  }]
 dyn_sideHairStrand_1_04_L [{
   shapeType=Capsule
   length=6.5531
   radius=1.7217
   shapeName=dyn_backHairStrand_3_01_capsule
   center=<6.5782 -0.4207 -0.566>
   rotation=<rot 87.161 179.9992 175.9302>
  }]
 dyn_sideHairStrand_1_04_R [{
   shapeType=Capsule
   length=7.543
   radius=1.86
   shapeName=dyn_backHairStrand_3_01_capsule
   center=<6.4596 0.6662 0.472>
   rotation=<rot 90 0 -14.1102>
  }]
 dyn_sideHairStrand_1_05_L [{shapeType=Capsule, length=7.5743, radius=1.5998, shapeName=dyn_backHairStrand_3_01_capsule, center=<5.9015 0 -0.1989>, rotation=<rot 90 0 0>}]
 dyn_sideHairStrand_1_05_R [{shapeType=Capsule, length=6.4709, radius=1.5496, shapeName=dyn_backHairStrand_3_01_capsule, center=<5.2672 0 0.2236>, rotation=<rot 90 0 0>}]
 dyn_sideHairStrand_1_06_L [{
   shapeType=Capsule
   length=6.5139
   radius=1.7695
   shapeName=dyn_backHairStrand_3_01_capsule
   center=<6.0724 -0.193 0.4021>
   rotation=<rot 90 -169.3803 -151.4675>
  }]
 dyn_sideHairStrand_1_06_R [{
   shapeType=Capsule
   length=6.7738
   radius=1.5882
   shapeName=dyn_backHairStrand_3_01_capsule
   center=<5.798 -0.2588 -0.0735>
   rotation=<rot 90 -178.2101 -165.6858>
  }]
 dyn_sideHairStrand_2_03_L [{shapeType=Capsule, length=9.513, radius=2.0466, shapeName=dyn_backHairStrand_3_01_capsule, center=<7.3664 0 0>, rotation=<rot 90 0 0>}]
 dyn_sideHairStrand_2_03_R [{
   shapeType=Capsule
   length=10.0244
   radius=1.7459
   shapeName=dyn_backHairStrand_3_01_capsule
   center=<7.1148 -0.0893 0.2233>
   rotation=<rot 90 0 0>
  }]
 dyn_sideHairStrand_2_04_L [{shapeType=Capsule, length=6.5624, radius=1.4207, shapeName=dyn_backHairStrand_3_01_capsule, center=<5.1409 -0.2994 0>, rotation=<rot 90 0 0>}]
 dyn_sideHairStrand_2_04_R [{
   shapeType=Capsule
   length=7.078
   radius=1.5601
   shapeName=dyn_backHairStrand_3_01_capsule
   center=<5.2426 -0.2591 0>
   rotation=<rot 90 180 179.0045>
  }]
 dyn_sideHairStrand_2_05_L [{
   shapeType=Capsule
   length=7.1436
   radius=1.6019
   shapeName=dyn_backHairStrand_3_01_capsule
   center=<5.2551 -0.3683 0.164>
   rotation=<rot 90 -14.0362 0.1045>
  }]
 dyn_sideHairStrand_2_05_R [{
   shapeType=Capsule
   length=7.568
   radius=1.7333
   shapeName=dyn_backHairStrand_3_01_capsule
   center=<6.0729 -0.4645 0.1193>
   rotation=<rot 90 -176.4237 -165.3673>
  }]
 dyn_sideHairStrand_2_06_L [{shapeType=Capsule, length=7.9206, radius=1.7021, shapeName=dyn_backHairStrand_3_01_capsule, center=<6.2374 0 0.5578>, rotation=<rot 90 0 0>}]
 dyn_sideHairStrand_2_06_R [{
   shapeType=Capsule
   length=8.6706
   radius=1.5909
   shapeName=dyn_backHairStrand_3_01_capsule
   center=<6.918 0.689 0.0189>
   rotation=<rot 90 161.0296 158.0384>
  }]
 head [
  {shapeType=Capsule, length=7.3573, radius=5.1096, center=<-6.5179 -0.9398 0>, rotation=<rot 90 0 -15.2774>}
  {shapeType=Capsule, length=8.9415, radius=9.4672, shapeName=head_capsule, center=<0.3399 -0.0104 0>, rotation=<rot 90 -165.9638 103.9436>}
 ]
 lowerarm_l [{shapeType=Capsule, length=14.7539, radius=4.5198, shapeName=lowerarm_l_capsule, center=<8.4672 0 0>, rotation=<rot 90 0 0>}]
 lowerarm_r [{shapeType=Capsule, length=14.7539, radius=4.5198, shapeName=lowerarm_l_capsule, center=<-8.4672 0 0>, rotation=<rot 90 0 0>}]
 spine_03 [{shapeType=Capsule, length=5.7181, radius=9.1839, shapeName=spine_03_capsule, center=<0.9858 4.2108 0>, rotation=<rot 0 -90 0>}]
 spine_04 [
  {shapeType=Capsule, length=7.3193, radius=10.802, shapeName=spine_04_capsule, center=<0.6851 0.3242 0>, rotation=<rot 0 -90 0>}
  {shapeType=Capsule, length=6.3248, radius=10.2903, shapeName=spine_04_capsule, center=<0.4949 4.4042 0>, rotation=<rot 0 -90 0>}
 ]
 spine_05 [{shapeType=Capsule, length=21.3906, radius=8.4737, shapeName=spine_05_capsule, center=<1.1438 -2.0253 0>, rotation=<rot 0 -90 0>}]
 upperarm_l [
  {shapeType=Capsule, length=18.7912, radius=5.3164, shapeName=upperarm_r_capsule, center=<13.3542 -0.4124 0.403>, rotation=<rot -90 180 180>}
  {shapeType=Capsule, length=7.9545, radius=5.2701, shapeName=upperarm_r_capsule, center=<4.263 -0.4124 0.2479>, rotation=<rot -90 180 180>}
 ]
 upperarm_r [
  {shapeType=Capsule, length=18.7912, radius=5.3164, shapeName=upperarm_r_capsule, center=<-13.3542 0.3257 0.403>, rotation=<rot -90 180 180>}
  {shapeType=Capsule, length=6.6076, radius=6.617, shapeName=upperarm_r_capsule, center=<-4.263 0.4124 -0.2479>, rotation=<rot -90 180 180>}
 ]
}
constraints [
 {bone1Name=dyn_sideHairStrand_1_03_L, bone2Name=base_sideHairStrand_1_03_L, swing1LimitDegrees=20, swing2LimitDegrees=20, twistLimitDegrees=2}
 {
  bone1Name=dyn_sideHairStrand_1_04_L
  bone2Name=dyn_sideHairStrand_1_03_L
  swing1LimitDegrees=45
  swing2LimitDegrees=25
  twistLimitDegrees=2
  twistMotion=Locked
 }
 {
  bone1Name=dyn_sideHairStrand_1_05_L
  bone2Name=dyn_sideHairStrand_1_04_L
  swing1LimitDegrees=70
  swing2LimitDegrees=70
  twistLimitDegrees=2
  twistMotion=Locked
 }
 {
  bone1Name=dyn_sideHairStrand_1_06_L
  bone2Name=dyn_sideHairStrand_1_05_L
  swing1LimitDegrees=60
  swing2LimitDegrees=60
  twistLimitDegrees=2
  twistMotion=Locked
 }
 {
  bone1Name=dyn_sideHairStrand_1_03_R
  bone2Name=base_sideHairStrand_1_03_R
  swing1LimitDegrees=22
  swing2LimitDegrees=20
  twistLimitDegrees=1
  twistMotion=Locked
 }
 {
  bone1Name=dyn_sideHairStrand_1_04_R
  bone2Name=dyn_sideHairStrand_1_03_R
  swing1LimitDegrees=45
  swing2LimitDegrees=45
  twistLimitDegrees=2
  twistMotion=Locked
 }
 {
  bone1Name=dyn_sideHairStrand_1_05_R
  bone2Name=dyn_sideHairStrand_1_04_R
  swing1LimitDegrees=45
  swing2LimitDegrees=45
  twistLimitDegrees=2
  twistMotion=Locked
 }
 {
  bone1Name=dyn_sideHairStrand_1_06_R
  bone2Name=dyn_sideHairStrand_1_05_R
  swing1LimitDegrees=45
  swing2LimitDegrees=45
  twistLimitDegrees=2
  twistMotion=Locked
 }
 {bone1Name=dyn_sideHairStrand_2_03_L, bone2Name=base_sideHairStrand_2_03_L, swing1LimitDegrees=20, swing2LimitDegrees=20, twistLimitDegrees=2}
 {
  bone1Name=dyn_sideHairStrand_2_04_L
  bone2Name=dyn_sideHairStrand_2_03_L
  swing1LimitDegrees=60
  swing2LimitDegrees=15
  twistLimitDegrees=2
  twistMotion=Locked
 }
 {
  bone1Name=dyn_sideHairStrand_2_05_L
  bone2Name=dyn_sideHairStrand_2_04_L
  swing1LimitDegrees=60
  swing2LimitDegrees=30
  twistLimitDegrees=2
  twistMotion=Locked
 }
 {
  bone1Name=dyn_sideHairStrand_2_06_L
  bone2Name=dyn_sideHairStrand_2_05_L
  swing1LimitDegrees=60
  swing2LimitDegrees=60
  twistLimitDegrees=2
  twistMotion=Locked
 }
 {
  bone1Name=dyn_sideHairStrand_2_03_R
  bone2Name=base_sideHairStrand_2_03_R
  swing1LimitDegrees=35
  swing2LimitDegrees=30
  twistLimitDegrees=1
  twistMotion=Locked
 }
 {
  bone1Name=dyn_sideHairStrand_2_04_R
  bone2Name=dyn_sideHairStrand_2_03_R
  swing1LimitDegrees=45
  swing2LimitDegrees=45
  twistLimitDegrees=2
  twistMotion=Locked
 }
 {
  bone1Name=dyn_sideHairStrand_2_05_R
  bone2Name=dyn_sideHairStrand_2_04_R
  swing1LimitDegrees=45
  swing2LimitDegrees=45
  twistLimitDegrees=2
  twistMotion=Locked
 }
 {
  bone1Name=dyn_sideHairStrand_2_06_R
  bone2Name=dyn_sideHairStrand_2_05_R
  swing1LimitDegrees=45
  swing2LimitDegrees=45
  twistLimitDegrees=2
  twistMotion=Locked
 }
 {
  bone1Name=head
  bone2Name=spine_05
  swing1LimitDegrees=45
  swing1Motion=Free
  swing2LimitDegrees=45
  swing2Motion=Free
  twistLimitDegrees=45
  twistMotion=Free
 }
 {
  bone1Name=spine_05
  bone2Name=spine_04
  swing1LimitDegrees=45
  swing1Motion=Free
  swing2LimitDegrees=45
  swing2Motion=Free
  twistLimitDegrees=45
  twistMotion=Free
 }
 {
  bone1Name=spine_04
  bone2Name=spine_03
  swing1LimitDegrees=45
  swing1Motion=Free
  swing2LimitDegrees=45
  swing2Motion=Free
  twistLimitDegrees=45
  twistMotion=Free
 }
 {
  bone1Name=clavicle_l
  bone2Name=spine_05
  swing1LimitDegrees=45
  swing1Motion=Free
  swing2LimitDegrees=45
  swing2Motion=Free
  twistLimitDegrees=45
  twistMotion=Free
 }
 {
  bone1Name=upperarm_l
  bone2Name=clavicle_l
  swing1LimitDegrees=45
  swing1Motion=Free
  swing2LimitDegrees=45
  swing2Motion=Free
  twistLimitDegrees=45
  twistMotion=Free
 }
 {
  bone1Name=clavicle_r
  bone2Name=spine_05
  swing1LimitDegrees=45
  swing1Motion=Free
  swing2LimitDegrees=45
  swing2Motion=Free
  twistLimitDegrees=45
  twistMotion=Free
 }
 {
  bone1Name=upperarm_r
  bone2Name=clavicle_r
  swing1LimitDegrees=45
  swing1Motion=Free
  swing2LimitDegrees=45
  swing2Motion=Free
  twistLimitDegrees=45
  twistMotion=Free
 }
 {
  bone1Name=base_sideHairStrand_2_03_R
  bone2Name=head
  swing1LimitDegrees=45
  swing1Motion=Free
  swing2LimitDegrees=45
  swing2Motion=Free
  twistLimitDegrees=45
  twistMotion=Free
 }
 {
  bone1Name=base_sideHairStrand_2_03_L
  bone2Name=head
  swing1LimitDegrees=45
  swing1Motion=Free
  swing2LimitDegrees=45
  swing2Motion=Free
  twistLimitDegrees=45
  twistMotion=Free
 }
 {
  bone1Name=base_sideHairStrand_1_03_R
  bone2Name=head
  swing1LimitDegrees=45
  swing1Motion=Free
  swing2LimitDegrees=45
  swing2Motion=Free
  twistLimitDegrees=45
  twistMotion=Free
 }
 {
  bone1Name=base_sideHairStrand_1_03_L
  bone2Name=head
  swing1LimitDegrees=45
  swing1Motion=Free
  swing2LimitDegrees=45
  swing2Motion=Free
  twistLimitDegrees=45
  twistMotion=Free
 }
 {
  bone1Name=lowerarm_l
  bone2Name=upperarm_l
  swing1LimitDegrees=45
  swing1Motion=Free
  swing2LimitDegrees=45
  swing2Motion=Free
  twistLimitDegrees=45
  twistMotion=Free
 }
 {
  bone1Name=lowerarm_r
  bone2Name=upperarm_r
  swing1LimitDegrees=45
  swing1Motion=Free
  swing2LimitDegrees=45
  swing2Motion=Free
  twistLimitDegrees=45
  twistMotion=Free
 }
]
