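# Control Rig for the CoachSalve character's face accessory part, applying corrective offsets to hair strand bones.
# Head tilt and spherical pose readers on head and upper arms compute alphas, including a multiplied look-up/two-handed-aim alpha.
# These, plus skydive-floating and parachute variables, weight transform offsets on back and side strands.
# Debug-gated controls can pose driver bones; shoulder parent constraints and arm-front offsets exist but are unconnected to execution.
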ControlRigBlueprint /BRCosmetics/Characters/Player/Female/Medium/Bodies/F_MED_HydraTrumpet_CoachSalve/Meshes/Parts/F_MED_HydraTrumpet_CoachSalve_FaceAcc_CtrlRig
graphs [
 New Function_ContainedGraph {nodes [
   n1 = RigVMFunctionEntryNode()
   RigVMFunction_ControlFlowBranch = &RigVMFunction_ControlFlowBranch(Condition=@n1.Debug, ExecuteContext=@n1.ExecuteContext)
   RigVMFunctionReturnNode(ExecuteContext=@RigVMFunction_ControlFlowBranch.Completed)
   RigUnit_GetTransform = &RigUnit_GetTransform(Item=(Type=Control,Name="head_ctrl"), bInitial=False)
   RigUnit_SetTransform = &RigUnit_SetTransform(
    Item=(Type=Bone,Name="head")
    Weight=1
    bInitial=False
    Value=@RigUnit_GetTransform.Transform
    ExecutePin=@RigVMFunction_ControlFlowBranch.True
   )
   > &RigUnit_SetTransform(
    Item=(Type=Bone,Name="clavicle_r")
    Weight=1
    bInitial=False
    Value=&RigUnit_GetTransform(Item=(Type=Control,Name="clavicle_r_ctrl"), bInitial=False).Transform
   )
   > &RigUnit_SetTransform(
    Item=(Type=Bone,Name="clavicle_l")
    Weight=1
    bInitial=False
    Value=&RigUnit_GetTransform(Item=(Type=Control,Name="clavicle_l_ctrl"), bInitial=False).Transform
   )
   > &RigUnit_SetTransform(
    Item=(Type=Bone,Name="upperarm_r")
    Weight=1
    bInitial=False
    Value=&RigUnit_GetTransform(Item=(Type=Control,Name="upperarm_r_ctrl"), bInitial=False).Transform
   )
   > &RigUnit_SetTransform(
    Item=(Type=Bone,Name="upperarm_l")
    Weight=1
    bInitial=False
    Value=&RigUnit_GetTransform(Item=(Type=Control,Name="upperarm_l_ctrl"), bInitial=False).Transform
   )
  ]}
 RigVMFunctionLibrary {nodes [Controls = RigVMCollapseNode(Debug=False, ExecuteContext=())]}
 RigVMFunction_Sequence_1_ContainedGraph {nodes [
   n1 = RigVMFunctionEntryNode()
   RigVMFunction_Sequence = &RigVMFunction_Sequence(ExecuteContext=@n1.ExecuteContext)
   n2 = &RigVMFunction_Sequence(ExecuteContext=@RigVMFunction_Sequence.B)
   RigVMFunctionReturnNode(A=@RigVMFunction_Sequence.A, B=@n2.A, C=@n2.B)
  ]}
 RigVMModel {nodes [
   &RigUnit_ModifyTransforms(
    ExecutePin=()
    ItemToModify="((Item=(Type=Bone,Name=\"dyn_sideHairStrand_1_05_R\"),Transform=(Rotation=(X=0,Y=0,Z=0,W=1),Translation=(X=0,Y=0,Z=2),Scale3D=(X=1,Y=1,Z=1))),(Item=(Type=Bone,Name=\"dyn_sideHairStrand_2_04_R\"),Transform=(Rotation=(X=0,Y=0,Z=0,W=1),Translation=(X=0,Y=0,Z=1),S ...(truncated)"
    WeightMinimum=0
    Weight=RigVMVariableNode() {variable=Upperarm Right Front Alpha}.Value
   )
   > &RigUnit_ModifyTransforms(
    ItemToModify="((Item=(Type=Bone,Name=\"dyn_sideHairStrand_1_05_L\"),Transform=(Rotation=(X=0,Y=0,Z=0,W=1),Translation=(X=0,Y=0,Z=-2),Scale3D=(X=1,Y=1,Z=1))),(Item=(Type=Bone,Name=\"dyn_sideHairStrand_2_04_R\"),Transform=(Rotation=(X=0,Y=0,Z=0,W=1),Translation=(X=0,Y=0,Z=0), ...(truncated)"
    WeightMinimum=0
    Weight=RigVMVariableNode() {variable=Upperarm Left Front Alpha}.Value
   )
   &RigUnit_ParentConstraint(
    AdvancedSettings=(InterpolationType=Average,RotationOrderForFilter=XZY)
    Child=(Type=Bone,Name="dyn_sideHairStrand_1_04_L")
    ExecutePin=()
    Filter=(TranslationFilter=(bX=True,bY=True,bZ=false),RotationFilter=(bX=false,bY=false,bZ=false),ScaleFilter=(bX=false,bY=false,bZ=false))
    Parents=((Item=(Type=Bone,Name="clavicle_l"),Weight=0.9),(Item=(Type=Bone,Name="spine_05"),Weight=0),(Item=(Type=Bone,Name="upperarm_l"),Weight=0.1))
    Weight=1
   )
   > &RigUnit_ParentConstraint(
    AdvancedSettings=(InterpolationType=Average,RotationOrderForFilter=XZY)
    Child=(Type=Bone,Name="dyn_sideHairStrand_2_04_L")
    Filter=(TranslationFilter=(bX=True,bY=True,bZ=false),RotationFilter=(bX=false,bY=false,bZ=false),ScaleFilter=(bX=false,bY=false,bZ=false))
    Parents=((Item=(Type=Bone,Name="clavicle_l"),Weight=0.5),(Item=(Type=Bone,Name="spine_05"),Weight=0),(Item=(Type=Bone,Name="upperarm_l"),Weight=0.5))
    Weight=1
   )
   > &RigUnit_ParentConstraint(
    AdvancedSettings=(InterpolationType=Average,RotationOrderForFilter=XZY)
    Child=(Type=Bone,Name="dyn_sideHairStrand_1_04_R")
    Filter=(TranslationFilter=(bX=True,bY=True,bZ=false),RotationFilter=(bX=false,bY=false,bZ=false),ScaleFilter=(bX=false,bY=false,bZ=false))
    Parents=((Item=(Type=Bone,Name="clavicle_r"),Weight=0.9),(Item=(Type=Bone,Name="spine_05"),Weight=0),(Item=(Type=Bone,Name="upperarm_r"),Weight=0.1))
    Weight=1
   )
   > &RigUnit_ParentConstraint(
    AdvancedSettings=(InterpolationType=Average,RotationOrderForFilter=XZY)
    Child=(Type=Bone,Name="dyn_sideHairStrand_2_04_R")
    Filter=(TranslationFilter=(bX=True,bY=True,bZ=false),RotationFilter=(bX=false,bY=false,bZ=false),ScaleFilter=(bX=false,bY=false,bZ=false))
    Parents=((Item=(Type=Bone,Name="clavicle_r"),Weight=0.9),(Item=(Type=Bone,Name="spine_05"),Weight=0),(Item=(Type=Bone,Name="upperarm_r"),Weight=0.1))
    Weight=1
   )
   RigUnit_BeginExecution = &RigUnit_BeginExecution()
   RigVMFunction_Sequence = RigVMAggregateNode(ExecuteContext=@RigUnit_BeginExecution.ExecutePin)
   Controls = RigVMFunctionReferenceNode(Debug=false, ExecuteContext=@RigVMFunction_Sequence.A)
   RigVMFunction_MathQuaternionToEuler = &RigVMFunction_MathQuaternionToEuler(
    RotationOrder=ZYX
    Value=&RigUnit_GetRelativeTransformForItem(Child=(Type=Bone,Name="head"), Parent=(Type=Bone,Name="spine_05"), bChildInitial=False, bParentInitial=False).RelativeTransform.Rotation
   )
   Remap = &RigVMFunction_MathDoubleRemap(SourceMaximum=-60, SourceMinimum=0, TargetMaximum=1, TargetMinimum=0, Value=@RigVMFunction_MathQuaternionToEuler.Result.Z)
   RigVMVariableNode(Value=@Remap.Result, ExecuteContext=@RigVMFunction_Sequence.B) {variable=headTiltUp Alpha}
   > n1 = &RigUnit_SphericalPoseReader(
    ActiveRegionScaleFactors=(PositiveWidth=0.5,NegativeWidth=3,PositiveHeight=1.5,NegativeHeight=1.5)
    ActiveRegionSize=0.1
    Debug=(bDrawDebug=false,bDraw2D=False,bDrawLocalAxes=False,DebugScale=25,DebugSegments=20,DebugThickness=0.25)
    DriverAxis=(X=-1,Y=0,Z=0)
    DriverItem=(Type=Bone,Name="upperarm_r")
    FalloffRegionScaleFactors=(PositiveWidth=1,NegativeWidth=1,PositiveHeight=1,NegativeHeight=1)
    FalloffSize=0.4
    FlipHeightScaling=False
    FlipWidthScaling=False
    OptionalParentItem=(Type=Bone,Name="None")
    RotationOffset=(X=90,Y=0,Z=180)
   )
   > RigVMVariableNode(Value=@n1.OutputParam) {variable=Upperarm Right Front Alpha}
   > n2 = &RigUnit_SphericalPoseReader(
    ActiveRegionScaleFactors=(PositiveWidth=0.5,NegativeWidth=3,PositiveHeight=1.5,NegativeHeight=1.5)
    ActiveRegionSize=0.1
    Debug=(bDrawDebug=false,bDraw2D=False,bDrawLocalAxes=False,DebugScale=25,DebugSegments=20,DebugThickness=0.25)
    DriverAxis=(X=1,Y=0,Z=0)
    DriverItem=(Type=Bone,Name="upperarm_l")
    FalloffRegionScaleFactors=(PositiveWidth=1,NegativeWidth=1,PositiveHeight=1,NegativeHeight=1)
    FalloffSize=0.4
    FlipHeightScaling=False
    FlipWidthScaling=False
    OptionalParentItem=(Type=Bone,Name="None")
    RotationOffset=(X=90,Y=0,Z=0)
   )
   > RigVMVariableNode(Value=@n2.OutputParam) {variable=Upperarm Left Front Alpha}
   > n3 = &RigUnit_SphericalPoseReader(
    ActiveRegionScaleFactors=(PositiveWidth=1,NegativeWidth=0.5,PositiveHeight=0.75,NegativeHeight=0.75)
    ActiveRegionSize=0.1
    Debug=(bDrawDebug=true,bDraw2D=False,bDrawLocalAxes=False,DebugScale=25,DebugSegments=20,DebugThickness=0.25)
    DriverAxis=(X=0,Y=1,Z=0)
    DriverItem=(Type=Bone,Name="head")
    FalloffRegionScaleFactors=(PositiveWidth=1,NegativeWidth=0.5,PositiveHeight=1,NegativeHeight=1)
    FalloffSize=0.2
    FlipHeightScaling=False
    FlipWidthScaling=False
    OptionalParentItem=(Type=Bone,Name="None")
    RotationOffset=(X=90,Y=0,Z=-30)
   )
   > RigVMVariableNode(Value=@n3.OutputParam) {variable=Head SPR Look Up Alpha}
   > n4 = &RigUnit_SphericalPoseReader(
    ActiveRegionScaleFactors=(PositiveWidth=1,NegativeWidth=1,PositiveHeight=1,NegativeHeight=1)
    ActiveRegionSize=0.025
    Debug=(bDrawDebug=true,bDraw2D=False,bDrawLocalAxes=False,DebugScale=25,DebugSegments=20,DebugThickness=0.25)
    DriverAxis=(X=-1,Y=0,Z=0)
    DriverItem=(Type=Bone,Name="upperarm_r")
    FalloffRegionScaleFactors=(PositiveWidth=1,NegativeWidth=1,PositiveHeight=1,NegativeHeight=1)
    FalloffSize=0.1
    FlipHeightScaling=False
    FlipWidthScaling=False
    OptionalParentItem=(Type=Bone,Name="None")
    RotationOffset=(X=-115,Y=0,Z=25)
   )
   > RigVMVariableNode(Value=@n4.OutputParam) {variable=Upperarm R SPR Aim Up Two Hands Alpha}
   Multiply = &RigVMFunction_MathDoubleMul(B=RigVMVariableNode() {variable=Upperarm R SPR Aim Up Two Hands Alpha}.Value, A=RigVMVariableNode() {variable=Head SPR Look Up Alpha}.Value)
   > RigVMVariableNode(Value=@Multiply.Result) {variable=Look Up and Left Two Hands Multiplied Alpha}
   &RigUnit_ModifyTransforms(
    ItemToModify="((Item=(Type=Bone,Name=\"base_backHairStrand_1_02\"),Transform=(Rotation=(X=0,Y=0,Z=-0.4226,W=0.9063),Translation=(X=0,Y=0,Z=0),Scale3D=(X=1,Y=1,Z=1))),(Item=(Type=Bone,Name=\"base_backHairStrand_2_03\"),Transform=(Rotation=(X=0,Y=0,Z=-0.4617,W=0.887),Translat ...(truncated)"
    WeightMinimum=0
    ExecutePin=@RigVMFunction_Sequence.C
    Weight=RigVMVariableNode() {variable=headTiltUp Alpha}.Value
   )
   > &RigUnit_ModifyTransforms(
    ItemToModify="((Item=(Type=Bone,Name=\"base_backHairStrand_1_02\"),Transform=(Rotation=(X=0,Y=0,Z=-0.0523,W=0.9986),Translation=(X=0,Y=0,Z=0),Scale3D=(X=1,Y=1,Z=1))),(Item=(Type=Bone,Name=\"base_backHairStrand_2_03\"),Transform=(Rotation=(X=-0.0014,Y=-0.0174,Z=-0.0784,W=0.9 ...(truncated)"
    WeightMinimum=0
    Weight=RigVMVariableNode() {variable=is Skydive Floating}.Value
   )
   > &RigUnit_ModifyTransforms(
    ItemToModify="((Item=(Type=Bone,Name=\"base_sideHairStrand_1_03_R\"),Transform=(Rotation=(X=0,Y=0,Z=-0.2164,W=0.9763),Translation=(X=0,Y=0,Z=0),Scale3D=(X=1,Y=1,Z=1))),(Item=(Type=Bone,Name=\"base_sideHairStrand_2_03_R\"),Transform=(Rotation=(X=0,Y=0,Z=-0.0436,W=0.999),Tran ...(truncated)"
    WeightMinimum=0
    Weight=RigVMVariableNode() {variable=Look Up and Left Two Hands Multiplied Alpha}.Value
   )
   > &RigUnit_ModifyTransforms(
    ItemToModify=((Item=(Type=Bone,Name="base_backHairStrand_5_03"),Transform=(Rotation=(X=0.0113,Y=0.0421,Z=-0.2586,W=0.965),Translation=(X=0,Y=0,Z=0),Scale3D=(X=1,Y=1,Z=1))))
    WeightMinimum=0
    Weight=RigVMVariableNode() {variable=is Parachute Opened}.Value
   )
   RigVMVariableNode() {variable=is Skydiving Alpha}
  ]}
]
hierarchy [
 root {loc=<0 0 0>, rot=<0 0 0 1>}
 pelvis {parent=root, loc=<0 -0.5301 98.0571>, rot=<0.0127 0.707 -0.0127 -0.707>}
 spine_01 {parent=pelvis, loc=<6.1208 -0.9328 0>, rot=<0 0 0.034 -0.9994>}
 spine_02 {parent=spine_01, loc=<6.5997 -0.2095 0>, rot=<0 0 0.0238 -0.9997>}
 spine_03 {parent=spine_02, loc=<6.4914 -0.1237 0>, rot=<0 0 0.0149 -0.9999>}
 spine_04 {parent=spine_03, loc=<6.9598 0.0794 0>, rot=<0 0 -0.0132 -0.9999>}
 spine_05 {parent=spine_04, loc=<6.9987 0 0>, rot=<0 0 -0.0015 -1>}
 neck_01 {parent=spine_05, loc=<11.3138 0.1529 0>, rot=<0 0 -0.2003 -0.9797>}
 neck_02 {parent=neck_01, loc=<4.5627 0 0>, rot=<0 0 0.0504 -0.9987>}
 head {parent=neck_02, loc=<5.3914 0 0>, rot=<0 0 0.115 -0.9934>}
 base_sideHairStrand_1_01_L {parent=head, loc=<19.6837 -0.5236 -7.0423>, rot=<0.5158 -0.7764 -0.357 -0.0619>}
 dyn_sideHairStrand_1_01_L {parent=base_sideHairStrand_1_01_L, loc=<0 0 0>, rot=<0 0 0 1>}
 dyn_sideHairStrand_1_02_L {parent=dyn_sideHairStrand_1_01_L, loc=<5.0059 0 0>, rot=<0.0474 0.0646 -0.2277 -0.9704>}
 base_sideHairStrand_1_03_L {parent=dyn_sideHairStrand_1_02_L, loc=<8.0087 1.4969 0.5047>, rot=<-0.4309 -0.0089 -0.3009 -0.8507>}
 dyn_sideHairStrand_1_03_L {parent=base_sideHairStrand_1_03_L, loc=<0 0 0>, rot=<0 0 0 1>}
 dyn_sideHairStrand_1_04_L {parent=dyn_sideHairStrand_1_03_L, loc=<15.3569 4.1489 -1.1648>, rot=<0.0797 -0.0466 -0.3796 -0.9205>}
 dyn_sideHairStrand_1_05_L {parent=dyn_sideHairStrand_1_04_L, loc=<11.042 0 0>, rot=<-0.0564 0.085 0.0329 -0.9942>}
 dyn_sideHairStrand_1_06_L {parent=dyn_sideHairStrand_1_05_L, loc=<10.6073 0 0>, rot=<-0.0219 -0.0049 0.1765 -0.984>}
 dyn_sideHairStrand_1_07_L {parent=dyn_sideHairStrand_1_06_L, loc=<11.7704 0 0>, rot=<0 0 -0.1481 -0.989>}
 base_backHairStrand_6_01 {parent=head, loc=<21.3173 -1.0069 -5.2672>, rot=<0.5344 -0.8147 -0.2155 -0.0655>}
 dyn_backHairStrand_6_01 {parent=base_backHairStrand_6_01, loc=<0 0 0>, rot=<0 0 0 1>}
 dyn_backHairStrand_6_02 {parent=dyn_backHairStrand_6_01, loc=<5.1843 0.0107 -0.1417>, rot=<0.0379 0.0421 -0.1233 -0.9907>}
 base_backHairStrand_6_03 {parent=dyn_backHairStrand_6_02, loc=<10.2063 3.4924 0.4035>, rot=<-0.2912 -0.0161 -0.4147 -0.8619>}
 dyn_backHairStrand_6_03 {parent=base_backHairStrand_6_03, loc=<0 0 0>, rot=<0 0 0 1>}
 dyn_backHairStrand_6_04 {parent=dyn_backHairStrand_6_03, loc=<13.0546 0 0>, rot=<0.0012 -0.0149 0.1084 -0.994>}
 dyn_backHairStrand_6_05 {parent=dyn_backHairStrand_6_04, loc=<11.3142 0 0>, rot=<0.0021 -0.014 -0.0774 -0.9969>}
 dyn_backHairStrand_6_06 {parent=dyn_backHairStrand_6_05, loc=<14.2921 0 0>, rot=<0 0 0 1>}
 base_backHairStrand_2_01 {parent=head, loc=<20.5116 -0.7802 -2.9874>, rot=<0.6684 -0.7372 -0.0843 -0.0523>}
 dyn_backHairStrand_2_01 {parent=base_backHairStrand_2_01, loc=<0 0 0>, rot=<0 0 0 1>}
 dyn_backHairStrand_2_02 {parent=dyn_backHairStrand_2_01, loc=<5.6904 0 0>, rot=<-0.0217 -0.0094 -0.2742 -0.9614>}
 base_backHairStrand_2_03 {parent=dyn_backHairStrand_2_02, loc=<8.5617 1.122 0>, rot=<-0.0808 -0.0196 -0.4213 -0.9031>}
 dyn_backHairStrand_2_03 {parent=base_backHairStrand_2_03, loc=<0 0 0>, rot=<0 0 0 1>}
 dyn_backHairStrand_2_04 {parent=dyn_backHairStrand_2_03, loc=<13.0976 0 0>, rot=<-0.0004 -0.0029 -0.0112 -0.9999>}
 dyn_backHairStrand_2_05 {parent=dyn_backHairStrand_2_04, loc=<10.5715 0 0>, rot=<0.0004 -0.026 0.0244 -0.9994>}
 dyn_backHairStrand_2_06 {parent=dyn_backHairStrand_2_05, loc=<10.0214 0 0>, rot=<-0.0003 0.009 0.0023 -1>}
 dyn_backHairStrand_2_07 {parent=dyn_backHairStrand_2_06, loc=<11.8621 0 0>, rot=<0 0 0 1>}
 base_backHairStrand_1_01 {parent=head, loc=<20.3786 -3.0393 -0.1251>, rot=<0.5374 -0.8432 -0.0001 -0.0143>}
 dyn_backHairStrand_1_01 {parent=base_backHairStrand_1_01, loc=<0 0 0>, rot=<0 0 0 1>}
 base_backHairStrand_1_02 {parent=dyn_backHairStrand_1_01, loc=<8.9823 0 0>, rot=<-0.0036 -0.0032 -0.2094 -0.9778>}
 dyn_backHairStrand_1_02 {parent=base_backHairStrand_1_02, loc=<0 0 0>, rot=<0 0 0 1>}
 dyn_backHairStrand_1_03 {parent=dyn_backHairStrand_1_02, loc=<12.1571 6.7623 0>, rot=<-0.0036 0.0011 -0.3326 -0.9431>}
 dyn_backHairStrand_1_04 {parent=dyn_backHairStrand_1_03, loc=<12.0141 0 0>, rot=<0.0004 -0.0035 -0.0322 -0.9995>}
 dyn_backHairStrand_1_05 {parent=dyn_backHairStrand_1_04, loc=<12.7533 0 0>, rot=<-0.0001 0.0118 0.0431 -0.999>}
 dyn_backHairStrand_1_06 {parent=dyn_backHairStrand_1_05, loc=<9.8904 0.8834 0>, rot=<-0.0006 -0.0283 -0.0763 -0.9967>}
 dyn_backHairStrand_1_07 {parent=dyn_backHairStrand_1_06, loc=<9.5398 0 0>, rot=<0 0 0 1>}
 base_backHairStrand_4_01 {parent=head, loc=<22.1741 -2.9404 0.0937>, rot=<0.7595 -0.6463 0.0082 0.0735>}
 dyn_backHairStrand_4_01 {parent=base_backHairStrand_4_01, loc=<0 0 0>, rot=<0 0 0 1>}
 dyn_backHairStrand_4_02 {parent=dyn_backHairStrand_4_01, loc=<7.5525 4.5996 0>, rot=<0.0356 0.0518 -0.4129 -0.9086>}
 dyn_backHairStrand_4_03 {parent=dyn_backHairStrand_4_02, loc=<11.2711 6.7911 0>, rot=<-0.0138 0.0361 -0.4035 -0.9142>}
 dyn_backHairStrand_4_04 {parent=dyn_backHairStrand_4_03, loc=<9.5224 0 0>, rot=<-0.002 0.0283 -0.0322 -0.9991>}
 dyn_backHairStrand_4_05 {parent=dyn_backHairStrand_4_04, loc=<9.4296 0 0>, rot=<0 0 0 1>}
 base_backHairStrand_3_01 {parent=head, loc=<20.9697 -0.8042 2.2754>, rot=<0.6469 -0.7546 0.083 0.0731>}
 dyn_backHairStrand_3_01 {parent=base_backHairStrand_3_01, loc=<0 0 0>, rot=<0 0 0 1>}
 dyn_backHairStrand_3_02 {parent=dyn_backHairStrand_3_01, loc=<5.6853 0 0>, rot=<-0.0392 -0.0137 -0.3293 -0.9433>}
 base_backHairStrand_3_03 {parent=dyn_backHairStrand_3_02, loc=<9.0435 -0.1273 -0.2069>, rot=<0.1484 0.0068 -0.3407 -0.9284>}
 dyn_backHairStrand_3_03 {parent=base_backHairStrand_3_03, loc=<0 0 0>, rot=<0 0 0 1>}
 dyn_backHairStrand_3_04 {parent=dyn_backHairStrand_3_03, loc=<13.0683 0 0>, rot=<0.0007 0.0293 0.005 -0.9996>}
 dyn_backHairStrand_3_05 {parent=dyn_backHairStrand_3_04, loc=<11.329 0 0>, rot=<-0.0018 0.0094 0.0187 -0.9998>}
 dyn_backHairStrand_3_06 {parent=dyn_backHairStrand_3_05, loc=<9.7521 0 0>, rot=<-0.0009 0.0146 -0.017 -0.9997>}
 dyn_backHairStrand_3_07 {parent=dyn_backHairStrand_3_06, loc=<8.8875 0 0>, rot=<-0.0001 0.0013 -0.002 -1>}
 dyn_backHairStrand_3_08 {parent=dyn_backHairStrand_3_07, loc=<8.101 0 0>, rot=<0 0 0 1>}
 base_sideHairStrand_1_01_R {parent=head, loc=<19.2682 -0.6482 7.0504>, rot=<0.5227 -0.7097 0.4712 0.0325>}
 dyn_sideHairStrand_1_01_R {parent=base_sideHairStrand_1_01_R, loc=<0 0 0>, rot=<0 0 0 1>}
 dyn_sideHairStrand_1_02_R {parent=dyn_sideHairStrand_1_01_R, loc=<5.6609 0 0>, rot=<0.0112 -0.1074 -0.3538 -0.929>}
 base_sideHairStrand_1_03_R {parent=dyn_sideHairStrand_1_02_R, loc=<7.4268 0 0>, rot=<0.4988 -0.0531 -0.1794 -0.8463>}
 dyn_sideHairStrand_1_03_R {parent=base_sideHairStrand_1_03_R, loc=<0 0 0>, rot=<0 0 0 1>}
 dyn_sideHairStrand_1_04_R {parent=dyn_sideHairStrand_1_03_R, loc=<15.2679 4.1696 1.2894>, rot=<-0.1349 0.046 -0.3062 -0.9412>}
 dyn_sideHairStrand_1_05_R {parent=dyn_sideHairStrand_1_04_R, loc=<11.0124 2.2147 0>, rot=<0.1233 -0.0803 -0.0874 -0.9853>}
 dyn_sideHairStrand_1_06_R {parent=dyn_sideHairStrand_1_05_R, loc=<9.0759 0 0>, rot=<-0.0308 -0.0148 0.1685 -0.9851>}
 dyn_sideHairStrand_1_07_R {parent=dyn_sideHairStrand_1_06_R, loc=<11.7611 0 0>, rot=<0 0 0 1>}
 base_sideHairStrand_2_01_L {parent=head, loc=<17.2352 0.1477 -8.7625>, rot=<0.3793 -0.7314 -0.5667 0.0031>}
 dyn_sideHairStrand_2_01_L {parent=base_sideHairStrand_2_01_L, loc=<0 0 0>, rot=<0 0 0 1>}
 dyn_sideHairStrand_2_02_L {parent=dyn_sideHairStrand_2_01_L, loc=<4.7456 0 0>, rot=<0.0085 0.2615 -0.1984 -0.9446>}
 base_sideHairStrand_2_03_L {parent=dyn_sideHairStrand_2_02_L, loc=<7.0523 0 0>, rot=<-0.6279 0.0032 -0.1641 -0.7608>}
 dyn_sideHairStrand_2_03_L {parent=base_sideHairStrand_2_03_L, loc=<0 0 0>, rot=<0 0 0 1>}
 dyn_sideHairStrand_2_04_L {parent=dyn_sideHairStrand_2_03_L, loc=<13.0245 0 0>, rot=<-0.0617 0.0197 -0.1209 -0.9905>}
 dyn_sideHairStrand_2_05_L {parent=dyn_sideHairStrand_2_04_L, loc=<8.547 0 0>, rot=<0.0272 0.0187 0.0074 -0.9994>}
 dyn_sideHairStrand_2_06_L {parent=dyn_sideHairStrand_2_05_L, loc=<9.0544 -1.8308 0.4454>, rot=<-0.0068 -0.0023 0.2483 -0.9687>}
 dyn_sideHairStrand_2_07_L {parent=dyn_sideHairStrand_2_06_L, loc=<11.8381 0 0>, rot=<0 0 0 1>}
 base_sideHairStrand_2_01_R {parent=head, loc=<16.5041 0.3036 8.3764>, rot=<0.4107 -0.6993 0.5849 0.011>}
 dyn_sideHairStrand_2_01_R {parent=base_sideHairStrand_2_01_R, loc=<0 0 0>, rot=<0 0 0 1>}
 dyn_sideHairStrand_2_02_R {parent=dyn_sideHairStrand_2_01_R, loc=<4.7678 0 0>, rot=<-0.0303 -0.2714 -0.2038 -0.9401>}
 base_sideHairStrand_2_03_R {parent=dyn_sideHairStrand_2_02_R, loc=<6.6031 0 0>, rot=<0.6878 -0.0095 -0.1804 -0.703>}
 dyn_sideHairStrand_2_03_R {parent=base_sideHairStrand_2_03_R, loc=<0 0 0>, rot=<0 0 0 1>}
 dyn_sideHairStrand_2_04_R {parent=dyn_sideHairStrand_2_03_R, loc=<13.0208 0 0>, rot=<0.0029 -0.0136 -0.1074 -0.9941>}
 dyn_sideHairStrand_2_05_R {parent=dyn_sideHairStrand_2_04_R, loc=<9.4744 0 0>, rot=<0.0012 -0.0372 0.0102 -0.9993>}
 dyn_sideHairStrand_2_06_R {parent=dyn_sideHairStrand_2_05_R, loc=<9.9649 -1.4888 0>, rot=<-0.0112 0.0125 0.2495 -0.9682>}
 dyn_sideHairStrand_2_07_R {parent=dyn_sideHairStrand_2_06_R, loc=<12.9613 0 0>, rot=<0 0 0 1>}
 base_backHairStrand_5_01 {parent=head, loc=<20.3956 0.0216 4.9328>, rot=<0.6658 -0.7217 0.1587 0.1035>}
 dyn_backHairStrand_5_01 {parent=base_backHairStrand_5_01, loc=<0 0 0>, rot=<0 0 0 1>}
 dyn_backHairStrand_5_02 {parent=dyn_backHairStrand_5_01, loc=<6.1836 0 0>, rot=<0 0 -0.4204 -0.9073>}
 base_backHairStrand_5_03 {parent=dyn_backHairStrand_5_02, loc=<8.2667 0 0>, rot=<0.1949 0.0068 -0.2832 -0.939>}
 dyn_backHairStrand_5_03 {parent=base_backHairStrand_5_03, loc=<0 0 0>, rot=<0 0 0 1>}
 dyn_backHairStrand_5_04 {parent=dyn_backHairStrand_5_03, loc=<14.5457 0 0>, rot=<0.0014 0.0039 0.0997 -0.995>}
 dyn_backHairStrand_5_05 {parent=dyn_backHairStrand_5_04, loc=<10.7487 0 0>, rot=<-0.0066 0.0191 -0.0254 -0.9995>}
 clavicle_l {parent=spine_05, loc=<5.4831 -0.1053 -3.2946>, rot=<0.0465 -0.7289 -0.009 -0.6829>}
 upperarm_l {parent=clavicle_l, loc=<9.8658 -1.2136 -0.4105>, rot=<0.0164 -0.3245 -0.0269 -0.9454>}
 lowerarm_l {parent=upperarm_l, loc=<26.465 -0.8345 0.259>, rot=<0.0036 0.0161 -0.2298 -0.9731>}
 clavicle_r {parent=spine_05, loc=<5.4831 -0.1053 3.2946>, rot=<0.7289 0.0465 -0.6829 0.009>}
 upperarm_r {parent=clavicle_r, loc=<-9.8658 1.2136 0.4105>, rot=<0.0164 -0.3245 -0.0269 -0.9454>}
 lowerarm_r {parent=upperarm_r, loc=<-26.465 0.8345 -0.259>, rot=<0.0036 0.0161 -0.2298 -0.9731>}
 head_ctrl {controlType=EULER_TRANSFORM, type=Control, loc=<0 0 0>, rot=<0 0 0 1>}
 clavicle_r_ctrl {controlType=EULER_TRANSFORM, type=Control, loc=<0 0 0>, rot=<0 0 0 1>}
 upperarm_r_ctrl {controlType=EULER_TRANSFORM, parent=clavicle_r_ctrl, type=Control, loc=<0 0 0>, rot=<0 0 0 1>}
 clavicle_l_ctrl {controlType=EULER_TRANSFORM, type=Control, loc=<0 0 0>, rot=<0 0 0 1>}
 upperarm_l_ctrl {controlType=EULER_TRANSFORM, parent=clavicle_l_ctrl, type=Control, loc=<0 0 0>, rot=<0 0 0 1>}
]
variables [
 Head SPR Look Up Alpha {type=double, default=0}
 Look Up and Left Two Hands Multiplied Alpha {type=double, default=0}
 Upperarm Left Front Alpha {type=double, default=0}
 Upperarm R SPR Aim Up Two Hands Alpha {type=double, default=0}
 Upperarm Right Front Alpha {type=double, default=0}
 headTiltUp Alpha {type=double, default=0}
 is Parachute Opened {type=double, default=0}
 is Skydive Floating {type=double, default=0}
 is Skydiving Alpha {type=double, default=0}
]
