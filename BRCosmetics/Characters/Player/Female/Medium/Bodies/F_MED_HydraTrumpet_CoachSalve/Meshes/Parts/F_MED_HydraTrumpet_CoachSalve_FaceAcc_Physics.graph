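# Physics asset for the CoachSalve female body's face accessory part, setting up six back hair strands for physics motion against the upper body.
# Capsules cover pelvis, spine, head, clavicles, upper arms and each strand bone.
# Strand chains hang from small base bodies tied back to the head, with tight twist and swing limits mostly widening toward the tips.
# Body and base joints rotate freely, as does an extra cross-link between strands 5 and 3.
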PhysicsAsset /BRCosmetics/Characters/Player/Female/Medium/Bodies/F_MED_HydraTrumpet_CoachSalve/Meshes/Parts/F_MED_HydraTrumpet_CoachSalve_FaceAcc_Physics
bodies {
 base_backHairStrand_1_02 [{shapeType=Capsule, length=0.505, radius=0.505, shapeName=base_backHairStrand_1_02_capsule, rotation=<rot 0 -90 0>}]
 base_backHairStrand_2_03 [{shapeType=Capsule, length=0.505, radius=0.505, shapeName=base_backHairStrand_2_03_capsule, rotation=<rot 0 -90 0>}]
 base_backHairStrand_3_01 [{shapeType=Capsule, length=0.505, radius=0.505, shapeName=base_backHairStrand_3_01_capsule, rotation=<rot 0 -90 0>}]
 base_backHairStrand_3_03 [{shapeType=Capsule, length=0.505, radius=0.505, shapeName=base_backHairStrand_3_03_capsule, rotation=<rot 0 -90 0>}]
 base_backHairStrand_4_01 [{shapeType=Capsule, length=0.505, radius=0.505, shapeName=base_backHairStrand_4_01_capsule, rotation=<rot 0 -90 0>}]
 base_backHairStrand_5_03 [{shapeType=Capsule, length=0.505, radius=0.505, shapeName=base_backHairStrand_5_03_capsule, rotation=<rot 0 -90 0>}]
 base_backHairStrand_6_03 [{shapeType=Capsule, length=0.505, radius=0.505, shapeName=base_backHairStrand_6_03_capsule, rotation=<rot 0 -90 0>}]
 clavicle_l [{shapeType=Capsule, length=3, radius=2, shapeName=clavicle_l_capsule, rotation=<rot -90 0 0>}]
 clavicle_r [{shapeType=Capsule, length=3, radius=2, shapeName=clavicle_r_capsule, rotation=<rot -90 0 0>}]
 dyn_backHairStrand_1_02 [{
   shapeType=Capsule
   length=10.2216
   radius=2
   shapeName=dyn_backHairStrand_1_02_capsule
   center=<6.5277 2.6256 0>
   rotation=<rot 90 180 145.6222>
  }]
 dyn_backHairStrand_1_03 [{shapeType=Capsule, length=8.2166, radius=2, shapeName=dyn_backHairStrand_1_03_capsule, center=<6.3549 -0.2526 0>, rotation=<rot 90 0 0>}]
 dyn_backHairStrand_1_04 [{shapeType=Capsule, length=8.6539, radius=2, shapeName=dyn_backHairStrand_1_04_capsule, center=<6.6291 0.196 0>, rotation=<rot 90 0 0>}]
 dyn_backHairStrand_1_05 [{
   shapeType=Capsule
   length=8.1253
   radius=2
   shapeName=dyn_backHairStrand_1_05_capsule
   center=<4.8414 0.3083 0>
   rotation=<rot 90 178.2101 173.0984>
  }]
 dyn_backHairStrand_1_06 [{
   shapeType=Capsule
   length=8.2943
   radius=2
   shapeName=dyn_backHairStrand_1_06_capsule
   center=<3.901 0.2065 0>
   rotation=<rot 90 180 176.2761>
  }]
 dyn_backHairStrand_2_03 [{shapeType=Capsule, length=10.1304, radius=2, shapeName=dyn_backHairStrand_2_03_capsule, center=<6.0938 -0.1394 0>, rotation=<rot 90 0 0>}]
 dyn_backHairStrand_2_04 [{shapeType=Capsule, length=9.6005, radius=2.2231, shapeName=dyn_backHairStrand_2_04_capsule, center=<5.372 0 -0.0988>, rotation=<rot 90 0 0>}]
 dyn_backHairStrand_2_05 [{
   shapeType=Capsule
   length=6.5627
   radius=2.037
   shapeName=dyn_backHairStrand_2_05_capsule
   center=<5.1092 0.6538 0.2298>
   rotation=<rot 90 180 -178.2755>
  }]
 dyn_backHairStrand_2_06 [{
   shapeType=Capsule
   length=9.1428
   radius=1.6765
   shapeName=dyn_backHairStrand_2_06_capsule
   center=<5.6128 -0.2034 -0.1548>
   rotation=<rot 90 0 0>
  }]
 dyn_backHairStrand_3_03 [{
   shapeType=Capsule
   length=10.209
   radius=1.8962
   shapeName=dyn_backHairStrand_3_03_capsule
   center=<6.7976 -0.3312 -0.1938>
   rotation=<rot 90 0 0>
  }]
 dyn_backHairStrand_3_04 [{
   shapeType=Capsule
   length=7.2711
   radius=1.7455
   shapeName=dyn_backHairStrand_3_04_capsule
   center=<6.0512 -0.0897 -0.1094>
   rotation=<rot 90 0 0>
  }]
 dyn_backHairStrand_3_05 [{shapeType=Capsule, length=6.2912, radius=1.8017, shapeName=dyn_backHairStrand_3_05_capsule, center=<5.2333 0.0975 0>, rotation=<rot 90 0 0>}]
 dyn_backHairStrand_3_06 [{shapeType=Capsule, length=6.7101, radius=2, shapeName=dyn_backHairStrand_3_06_capsule, center=<4.8651 -0.0392 0>, rotation=<rot 90 0 0>}]
 dyn_backHairStrand_3_07 [{shapeType=Capsule, length=5.1927, radius=2, shapeName=dyn_backHairStrand_3_07_capsule, center=<3.9878 0 0>, rotation=<rot 90 0 0>}]
 dyn_backHairStrand_4_01 [{shapeType=Capsule, length=1, radius=2, shapeName=dyn_backHairStrand_3_01_capsule, center=<2 0 0>, rotation=<rot 90 0 0>}]
 dyn_backHairStrand_4_02 [{
   shapeType=Capsule
   length=7.5746
   radius=2
   shapeName=dyn_backHairStrand_3_01_capsule
   center=<7.0304 2.2922 -0.1926>
   rotation=<rot 90 153.4349 118.3243>
  }]
 dyn_backHairStrand_4_03 [{shapeType=Capsule, length=7.4897, radius=1.7208, shapeName=dyn_backHairStrand_3_01_capsule, center=<4.7048 0 0.0283>, rotation=<rot 90 0 0>}]
 dyn_backHairStrand_4_04 [{shapeType=Capsule, length=7.0605, radius=1.4641, shapeName=dyn_backHairStrand_3_01_capsule, center=<4.6778 0 0.2152>, rotation=<rot 90 0 0>}]
 dyn_backHairStrand_5_03 [{
   shapeType=Capsule
   length=10.1626
   radius=2
   shapeName=dyn_backHairStrand_3_01_capsule
   center=<7.6417 -0.3062 -0.1139>
   rotation=<rot 90 0 0>
  }]
 dyn_backHairStrand_5_04 [{shapeType=Capsule, length=7.1905, radius=1.8197, shapeName=dyn_backHairStrand_3_01_capsule, center=<5.8876 0 0>, rotation=<rot 90 0 0>}]
 dyn_backHairStrand_5_05 [{shapeType=Capsule, length=7.8997, radius=2, shapeName=dyn_backHairStrand_3_01_capsule, center=<6.276 -0.3832 0>, rotation=<rot 90 0 0>}]
 dyn_backHairStrand_6_03 [{shapeType=Capsule, length=9.2821, radius=2, shapeName=dyn_backHairStrand_3_01_capsule, center=<7.1496 0 -0.2894>, rotation=<rot 90 0 0>}]
 dyn_backHairStrand_6_04 [{shapeType=Capsule, length=8.0002, radius=2, shapeName=dyn_backHairStrand_3_01_capsule, center=<6.5021 0.2346 0>, rotation=<rot 90 0 0>}]
 dyn_backHairStrand_6_05 [{
   shapeType=Capsule
   length=9.4559
   radius=2
   shapeName=dyn_backHairStrand_3_01_capsule
   center=<8.0842 -0.5029 -0.4004>
   rotation=<rot 90 0 0>
  }]
 head [
  {shapeType=Capsule, length=9, radius=4.6597, center=<-4.3999 -0.3613 0>, rotation=<rot 90 0 -15.2774>}
  {shapeType=Capsule, length=7.1512, radius=11.2575, shapeName=head_capsule, center=<6.797 0 0>, rotation=<rot 90 0 0>}
 ]
 pelvis [{shapeType=Capsule, length=19.6732, radius=11.9647, shapeName=pelvis_capsule, center=<2.3384 -2.276 0>, rotation=<rot 0 -90 0>}]
 spine_02 [{shapeType=Capsule, length=12.8951, radius=8.5251, shapeName=spine_02_capsule, center=<0 0.4276 0>, rotation=<rot 0 -90 0>}]
 spine_03 [{shapeType=Capsule, length=7.9381, radius=9.1839, shapeName=spine_03_capsule, center=<0 0.4783 0>, rotation=<rot 0 -90 0>}]
 spine_04 [{shapeType=Capsule, length=28.2459, radius=10.802, shapeName=spine_04_capsule, center=<0.6851 0.5174 0>, rotation=<rot 0 -90 0>}]
 spine_05 [
  {shapeType=Capsule, length=16.7857, radius=9, shapeName=spine_05_capsule, center=<1.4288 -0.4421 0>, rotation=<rot 0 -90 0>}
  {shapeType=Capsule, length=58.767, radius=10.2781, shapeName=spine_05_capsule, center=<17.1894 10.852 0>}
 ]
 upperarm_l [
  {shapeType=Capsule, length=18.7912, radius=5.3164, shapeName=upperarm_r_capsule, center=<13.3542 -0.4124 0.403>, rotation=<rot -90 180 180>}
  {shapeType=Capsule, length=8.1571, radius=7.0833, shapeName=upperarm_r_capsule, center=<4.4253 -0.4124 0.8047>, rotation=<rot -90 0 0>}
 ]
 upperarm_r [
  {shapeType=Capsule, length=18.7912, radius=5.3164, shapeName=upperarm_r_capsule, center=<-13.3542 0.3257 0.403>, rotation=<rot -90 180 180>}
  {shapeType=Capsule, length=8.3871, radius=6.8533, shapeName=upperarm_r_capsule, center=<-4.4253 0.4124 0.8047>, rotation=<rot -90 0 0>}
 ]
}
constraints [
 {bone1Name=dyn_backHairStrand_1_02, bone2Name=base_backHairStrand_1_02, swing1LimitDegrees=10, swing2LimitDegrees=7, twistLimitDegrees=2}
 {bone1Name=dyn_backHairStrand_1_03, bone2Name=dyn_backHairStrand_1_02, swing1LimitDegrees=30, swing2LimitDegrees=15, twistLimitDegrees=2}
 {bone1Name=dyn_backHairStrand_1_04, bone2Name=dyn_backHairStrand_1_03, swing1LimitDegrees=45, swing2LimitDegrees=25, twistLimitDegrees=2}
 {bone1Name=dyn_backHairStrand_1_05, bone2Name=dyn_backHairStrand_1_04, swing1LimitDegrees=45, swing2LimitDegrees=35, twistLimitDegrees=2}
 {bone1Name=dyn_backHairStrand_1_06, bone2Name=dyn_backHairStrand_1_05, swing1LimitDegrees=45, swing2LimitDegrees=45, twistLimitDegrees=2}
 {bone1Name=dyn_backHairStrand_2_03, bone2Name=base_backHairStrand_2_03, swing1LimitDegrees=10, swing2LimitDegrees=10, twistLimitDegrees=2}
 {bone1Name=dyn_backHairStrand_2_04, bone2Name=dyn_backHairStrand_2_03, swing1LimitDegrees=30, swing2LimitDegrees=30, twistLimitDegrees=2}
 {bone1Name=dyn_backHairStrand_2_05, bone2Name=dyn_backHairStrand_2_04, swing1LimitDegrees=45, swing2LimitDegrees=45, twistLimitDegrees=2}
 {bone1Name=dyn_backHairStrand_2_06, bone2Name=dyn_backHairStrand_2_05, swing1LimitDegrees=45, swing2LimitDegrees=45, twistLimitDegrees=2}
 {
  bone1Name=base_backHairStrand_3_01
  bone2Name=head
  swing1LimitDegrees=45
  swing1Motion=Free
  swing2LimitDegrees=45
  swing2Motion=Free
  twistLimitDegrees=45
  twistMotion=Free
 }
 {bone1Name=dyn_backHairStrand_3_03, bone2Name=base_backHairStrand_3_03, swing1LimitDegrees=10, swing2LimitDegrees=10, twistLimitDegrees=2}
 {bone1Name=dyn_backHairStrand_3_04, bone2Name=dyn_backHairStrand_3_03, swing1LimitDegrees=30, swing2LimitDegrees=30, twistLimitDegrees=2}
 {bone1Name=dyn_backHairStrand_3_05, bone2Name=dyn_backHairStrand_3_04, swing1LimitDegrees=45, swing2LimitDegrees=45, twistLimitDegrees=2}
 {bone1Name=dyn_backHairStrand_3_06, bone2Name=dyn_backHairStrand_3_05, swing1LimitDegrees=45, swing2LimitDegrees=45, twistLimitDegrees=2}
 {bone1Name=dyn_backHairStrand_3_07, bone2Name=dyn_backHairStrand_3_06, swing1LimitDegrees=45, swing2LimitDegrees=45, twistLimitDegrees=2}
 {
  bone1Name=base_backHairStrand_4_01
  bone2Name=head
  swing1LimitDegrees=45
  swing1Motion=Free
  swing2LimitDegrees=45
  swing2Motion=Free
  twistLimitDegrees=45
  twistMotion=Free
 }
 {bone1Name=dyn_backHairStrand_4_01, bone2Name=base_backHairStrand_4_01, swing1LimitDegrees=25, swing2LimitDegrees=25, twistLimitDegrees=1}
 {bone1Name=dyn_backHairStrand_4_02, bone2Name=dyn_backHairStrand_4_01, swing1LimitDegrees=10, swing2LimitDegrees=6, twistLimitDegrees=2}
 {bone1Name=dyn_backHairStrand_4_03, bone2Name=dyn_backHairStrand_4_02, swing1LimitDegrees=30, swing2LimitDegrees=15, twistLimitDegrees=2}
 {bone1Name=dyn_backHairStrand_4_04, bone2Name=dyn_backHairStrand_4_03, swing1LimitDegrees=45, swing2LimitDegrees=25, twistLimitDegrees=2}
 {bone1Name=dyn_backHairStrand_5_03, bone2Name=base_backHairStrand_5_03, swing1LimitDegrees=10, swing2LimitDegrees=5, twistLimitDegrees=2}
 {bone1Name=dyn_backHairStrand_5_04, bone2Name=dyn_backHairStrand_5_03, swing1LimitDegrees=30, swing2LimitDegrees=15, twistLimitDegrees=2}
 {bone1Name=dyn_backHairStrand_5_05, bone2Name=dyn_backHairStrand_5_04, swing1LimitDegrees=45, swing2LimitDegrees=25, twistLimitDegrees=2}
 {bone1Name=dyn_backHairStrand_6_03, bone2Name=base_backHairStrand_6_03, swing1LimitDegrees=10, swing2LimitDegrees=10, twistLimitDegrees=2}
 {bone1Name=dyn_backHairStrand_6_04, bone2Name=dyn_backHairStrand_6_03, swing1LimitDegrees=30, swing2LimitDegrees=30, twistLimitDegrees=2}
 {bone1Name=dyn_backHairStrand_6_05, bone2Name=dyn_backHairStrand_6_04, swing1LimitDegrees=45, swing2LimitDegrees=45, twistLimitDegrees=2}
 {
  bone1Name=head
  bone2Name=spine_05
  swing1LimitDegrees=45
  swing1Motion=Free
  swing2LimitDegrees=45
  swing2Motion=Free
  twistLimitDegrees=45
  twistMotion=Free
 }
 {
  bone1Name=spine_05
  bone2Name=spine_04
  swing1LimitDegrees=45
  swing1Motion=Free
  swing2LimitDegrees=45
  swing2Motion=Free
  twistLimitDegrees=45
  twistMotion=Free
 }
 {
  bone1Name=spine_04
  bone2Name=spine_03
  swing1LimitDegrees=45
  swing1Motion=Free
  swing2LimitDegrees=45
  swing2Motion=Free
  twistLimitDegrees=45
  twistMotion=Free
 }
 {
  bone1Name=spine_03
  bone2Name=spine_02
  swing1LimitDegrees=45
  swing1Motion=Free
  swing2LimitDegrees=45
  swing2Motion=Free
  twistLimitDegrees=45
  twistMotion=Free
 }
 {
  bone1Name=clavicle_l
  bone2Name=spine_05
  swing1LimitDegrees=45
  swing1Motion=Free
  swing2LimitDegrees=45
  swing2Motion=Free
  twistLimitDegrees=45
  twistMotion=Free
 }
 {
  bone1Name=upperarm_l
  bone2Name=clavicle_l
  swing1LimitDegrees=45
  swing1Motion=Free
  swing2LimitDegrees=45
  swing2Motion=Free
  twistLimitDegrees=45
  twistMotion=Free
 }
 {
  bone1Name=clavicle_r
  bone2Name=spine_05
  swing1LimitDegrees=45
  swing1Motion=Free
  swing2LimitDegrees=45
  swing2Motion=Free
  twistLimitDegrees=45
  twistMotion=Free
 }
 {
  bone1Name=upperarm_r
  bone2Name=clavicle_r
  swing1LimitDegrees=45
  swing1Motion=Free
  swing2LimitDegrees=45
  swing2Motion=Free
  twistLimitDegrees=45
  twistMotion=Free
 }
 {
  bone1Name=spine_02
  bone2Name=pelvis
  swing1LimitDegrees=45
  swing1Motion=Free
  swing2LimitDegrees=45
  swing2Motion=Free
  twistLimitDegrees=45
  twistMotion=Free
 }
 {
  bone1Name=base_backHairStrand_1_02
  bone2Name=head
  swing1LimitDegrees=45
  swing1Motion=Free
  swing2LimitDegrees=45
  swing2Motion=Free
  twistLimitDegrees=45
  twistMotion=Free
 }
 {
  bone1Name=base_backHairStrand_2_03
  bone2Name=head
  swing1LimitDegrees=45
  swing1Motion=Free
  swing2LimitDegrees=45
  swing2Motion=Free
  twistLimitDegrees=45
  twistMotion=Free
 }
 {
  bone1Name=base_backHairStrand_3_03
  bone2Name=base_backHairStrand_3_01
  swing1LimitDegrees=45
  swing1Motion=Free
  swing2LimitDegrees=45
  swing2Motion=Free
  twistLimitDegrees=45
  twistMotion=Free
 }
 {
  bone1Name=base_backHairStrand_5_03
  bone2Name=head
  swing1LimitDegrees=45
  swing1Motion=Free
  swing2LimitDegrees=45
  swing2Motion=Free
  twistLimitDegrees=45
  twistMotion=Free
 }
 {
  bone1Name=base_backHairStrand_6_03
  bone2Name=head
  swing1LimitDegrees=45
  swing1Motion=Free
  swing2LimitDegrees=45
  swing2Motion=Free
  twistLimitDegrees=45
  twistMotion=Free
 }
 {
  bone1Name=dyn_backHairStrand_5_04
  bone2Name=dyn_backHairStrand_3_04
  swing1LimitDegrees=45
  swing1Motion=Free
  swing2LimitDegrees=45
  swing2Motion=Free
  twistLimitDegrees=45
  twistMotion=Free
 }
]
