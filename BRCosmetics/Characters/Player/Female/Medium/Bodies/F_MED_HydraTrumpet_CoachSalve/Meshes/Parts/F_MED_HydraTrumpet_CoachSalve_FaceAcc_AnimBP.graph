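# Post-body AnimBP for the CoachSalve outfit's face accessory mesh, animating its hair on the player character.
# Copied parent and body poses, blended at the upper arms, feed a Control Rig driven by skydive and parachute states.
# Normally two head-space RigidBody passes simulate dynamics, the first using state-dependent gravity.
# With backbling or cape, only the front-hair simulation runs, and weighted constraints bind back-hair strands to head, spine and clavicles.
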AnimBlueprint /BRCosmetics/Characters/Player/Female/Medium/Bodies/F_MED_HydraTrumpet_CoachSalve/Meshes/Parts/F_MED_HydraTrumpet_CoachSalve_FaceAcc_AnimBP
target_skeleton=&/F_MED_HydraTrumpet_CoachSalve_FaceAcc_Skeleton
graphs [
 AnimGraph {nodes [
   Output Pose(Result=Blend Poses by bool(
     BlendPose_0=Use cached pose 'Backpack Setup'()
     BlendPose_1=Use cached pose 'Dynamics Setup'()
     bActiveValue=Get is Wearing Backbling or Cape() -> (is Wearing Backbling or Cape:bool)
    ) {blendType=Linear}) {layerGroup=None}
   n1 = Get Body_Mesh()
   Save cached pose 'Start Pose'(Pose=Layered blend per bone(
     BasePose=Copy Pose From Mesh() {bUseAttachedParent=true}
     BlendPoses_0=Copy Pose From Mesh(SourceMeshComponent=@n1) {bCopyCurves=true, bUseAttachedParent=true}
    ) {layerSetup [{branchFilters [{boneName=upperarm_l}, {boneName=upperarm_r}]}]}) {cacheName=Start Pose}
   n2 = Control Rig!(
    Source=Use cached pose 'Start Pose'()
    Alpha
    is Skydiving Alpha:double="Get is Skydiving "() -> ("is Skydiving ":double)
    is Parachute Opened:double=Get IsParachuteOpened() -> (IsParachuteOpened:double)
    is Skydive Floating:double=Get IsSkydiveFloating() -> (IsSkydiveFloating:double)
   ) -> (Pose) {
    alphaBoolBlend {blendOption=Linear}
    controlRigAssetReference {blueprintRigClass=&/F_MED_HydraTrumpet_CoachSalve_FaceAcc_CtrlRig.F_MED_HydraTrumpet_CoachSalve_FaceAcc_CtrlRig_C}
   }
   Save cached pose 'CR Pose'(Pose=@n2) {cacheName=CR Pose}
   n3 = RigidBody!(
    ComponentPose=Local To Component(LocalPose=Use cached pose 'CR Pose'())
    Alpha
    OverrideWorldGravity:Vector=Get GravityOverride()
    ExternalForce
   ) -> (Pose) {
    bClampLinearTranslationLimitToRefPose=true
    bOverrideWorldGravity=true
    overridePhysicsAsset=&/F_MED_HydraTrumpet_CoachSalve_FaceAcc_Physics
    simulationSpace=BaseBoneSpace
    alphaBoolBlend {blendOption=Linear}
    baseBoneRef {boneName=head}
    simSpaceSettings {
     dampingAlpha=0
     maxAngularAcceleration=16
     maxAngularVelocity=8
     maxLinearAcceleration=800
     maxLinearVelocity=600
     worldAlpha=1
     externalLinearDragV=<2 0.2 0.2>
    }
   }
   n4 = RigidBody(ComponentPose=@n3, OverrideWorldGravity="0,0,-980") {
    bClampLinearTranslationLimitToRefPose=true
    bOverrideWorldGravity=true
    overridePhysicsAsset=&/F_MED_HydraTrumpet_CS_FaceAcc_FrHair_Physics
    simulationSpace=BaseBoneSpace
    alphaBoolBlend {blendOption=Linear}
    baseBoneRef {boneName=head}
    simSpaceSettings {
     dampingAlpha=0
     maxAngularAcceleration=16
     maxAngularVelocity=8
     maxLinearAcceleration=600
     maxLinearVelocity=700
     velocityScaleZ=0.32
     worldAlpha=1
     externalLinearDragV=<2 0.2 0.2>
    }
   }
   Save cached pose 'Dynamics Setup'(Pose=Component To Local(ComponentPose=@n4)) {cacheName=Dynamics Setup}
   n5 = RigidBody(ComponentPose=Local To Component(LocalPose=Use cached pose 'CR Pose'()), OverrideWorldGravity="0,0,-980") {
    bClampLinearTranslationLimitToRefPose=true
    bOverrideWorldGravity=true
    overridePhysicsAsset=&F_MED_HydraTrumpet_CS_FaceAcc_FrHair_Physics
    simulationSpace=BaseBoneSpace
    alphaBoolBlend {blendOption=Linear}
    baseBoneRef {boneName=head}
    simSpaceSettings {
     dampingAlpha=0
     maxAngularAcceleration=16
     maxAngularVelocity=8
     maxLinearAcceleration=600
     maxLinearVelocity=700
     velocityScaleZ=0.32
     worldAlpha=1
     externalLinearDragV=<2 0.2 0.2>
    }
   }
   n6 = "Constraint - dyn_backHairStrand_1_02 (P)"(ComponentPose=@n5, Alpha=1, ConstraintWeights_0=1) {
    alphaBoolBlend {blendOption=Linear}
    boneToModify {boneName=dyn_backHairStrand_1_02}
    constraintSetup [{targetBone {boneName=head}}]
    constraintWeights [1]
   }
   n7 = "Constraint - dyn_backHairStrand_1_03 (PP)"!(ComponentPose=@n6, Alpha=1, ConstraintWeights_0=0.25, ConstraintWeights_1:float=0.75) -> (Pose) {
    alphaBoolBlend {blendOption=Linear}
    boneToModify {boneName=dyn_backHairStrand_1_03}
    constraintSetup [{targetBone {boneName=spine_05}}, {targetBone {boneName=head}}]
    constraintWeights [1,1]
   }
   n8 = "Constraint - dyn_backHairStrand_1_04 (PP)"!(ComponentPose=@n7, Alpha=1, ConstraintWeights_0=0.75, ConstraintWeights_1:float=0.25) -> (Pose) {
    alphaBoolBlend {blendOption=Linear}
    boneToModify {boneName=dyn_backHairStrand_1_04}
    constraintSetup [{targetBone {boneName=spine_05}}, {targetBone {boneName=head}}]
    constraintWeights [1,1]
   }
   n9 = "Constraint - dyn_backHairStrand_1_05 (P)"(ComponentPose=@n8, Alpha=1, ConstraintWeights_0=1) {
    alphaBoolBlend {blendOption=Linear}
    boneToModify {boneName=dyn_backHairStrand_1_05}
    constraintSetup [{targetBone {boneName=spine_05}}]
    constraintWeights [1]
   }
   n10 = "Constraint - dyn_backHairStrand_1_06 (P)"(ComponentPose=@n9, Alpha=1, ConstraintWeights_0=1) {
    alphaBoolBlend {blendOption=Linear}
    boneToModify {boneName=dyn_backHairStrand_1_06}
    constraintSetup [{targetBone {boneName=spine_05}}]
    constraintWeights [1]
   }
   n11 = "Constraint - dyn_backHairStrand_2_03 (P)"(ComponentPose=@n10, Alpha=1, ConstraintWeights_0=1) {
    alphaBoolBlend {blendOption=Linear}
    boneToModify {boneName=dyn_backHairStrand_2_03}
    constraintSetup [{targetBone {boneName=head}}]
    constraintWeights [1]
   }
   n12 = "Constraint - dyn_backHairStrand_2_04 (PP)"!(ComponentPose=@n11, Alpha=1, ConstraintWeights_0=0.25, ConstraintWeights_1:float=0.75) -> (Pose) {
    alphaBoolBlend {blendOption=Linear}
    boneToModify {boneName=dyn_backHairStrand_2_04}
    constraintSetup [{targetBone {boneName=spine_05}}, {targetBone {boneName=head}}]
    constraintWeights [1,1]
   }
   n13 = "Constraint - dyn_backHairStrand_2_05 (PP)"!(ComponentPose=@n12, Alpha=1, ConstraintWeights_0=0.75, ConstraintWeights_1:float=0.25) -> (Pose) {
    alphaBoolBlend {blendOption=Linear}
    boneToModify {boneName=dyn_backHairStrand_2_05}
    constraintSetup [{targetBone {boneName=spine_05}}, {targetBone {boneName=head}}]
    constraintWeights [1,1]
   }
   n14 = "Constraint - dyn_backHairStrand_2_06 (P)"(ComponentPose=@n13, Alpha=1, ConstraintWeights_0=1) {
    alphaBoolBlend {blendOption=Linear}
    boneToModify {boneName=dyn_backHairStrand_2_06}
    constraintSetup [{targetBone {boneName=spine_05}}]
    constraintWeights [1]
   }
   n15 = "Constraint - dyn_backHairStrand_3_03 (P)"(ComponentPose=@n14, Alpha=1, ConstraintWeights_0=1) {
    alphaBoolBlend {blendOption=Linear}
    boneToModify {boneName=dyn_backHairStrand_3_03}
    constraintSetup [{targetBone {boneName=head}}]
    constraintWeights [1]
   }
   n16 = "Constraint - dyn_backHairStrand_3_04 (PP)"!(ComponentPose=@n15, Alpha=1, ConstraintWeights_0=0.25, ConstraintWeights_1:float=0.75) -> (Pose) {
    alphaBoolBlend {blendOption=Linear}
    boneToModify {boneName=dyn_backHairStrand_3_04}
    constraintSetup [{targetBone {boneName=spine_05}}, {targetBone {boneName=head}}]
    constraintWeights [1,1]
   }
   n17 = "Constraint - dyn_backHairStrand_3_05 (PP)"!(ComponentPose=@n16, Alpha=1, ConstraintWeights_0=0.75, ConstraintWeights_1:float=0.25) -> (Pose) {
    alphaBoolBlend {blendOption=Linear}
    boneToModify {boneName=dyn_backHairStrand_3_05}
    constraintSetup [{targetBone {boneName=spine_05}}, {targetBone {boneName=head}}]
    constraintWeights [1,1]
   }
   n18 = "Constraint - dyn_backHairStrand_3_06 (P)"(ComponentPose=@n17, Alpha=1, ConstraintWeights_0=1) {
    alphaBoolBlend {blendOption=Linear}
    boneToModify {boneName=dyn_backHairStrand_3_06}
    constraintSetup [{targetBone {boneName=spine_05}}]
    constraintWeights [1]
   }
   n19 = "Constraint - dyn_backHairStrand_3_07 (P)"(ComponentPose=@n18, Alpha=1, ConstraintWeights_0=1) {
    alphaBoolBlend {blendOption=Linear}
    boneToModify {boneName=dyn_backHairStrand_3_07}
    constraintSetup [{targetBone {boneName=spine_05}}]
    constraintWeights [1]
   }
   n20 = "Constraint - dyn_backHairStrand_4_01 (P)"(ComponentPose=@n19, Alpha=1, ConstraintWeights_0=1) {
    alphaBoolBlend {blendOption=Linear}
    boneToModify {boneName=dyn_backHairStrand_4_01}
    constraintSetup [{targetBone {boneName=head}}]
    constraintWeights [1]
   }
   n21 = "Constraint - dyn_backHairStrand_4_02 (P)"(ComponentPose=@n20, Alpha=1, ConstraintWeights_0=1) {
    alphaBoolBlend {blendOption=Linear}
    boneToModify {boneName=dyn_backHairStrand_4_02}
    constraintSetup [{targetBone {boneName=head}}]
    constraintWeights [1]
   }
   n22 = "Constraint - dyn_backHairStrand_4_03 (P)"(ComponentPose=@n21, Alpha=1, ConstraintWeights_0=1) {
    alphaBoolBlend {blendOption=Linear}
    boneToModify {boneName=dyn_backHairStrand_4_03}
    constraintSetup [{targetBone {boneName=head}}]
    constraintWeights [1]
   }
   n23 = "Constraint - dyn_backHairStrand_4_04 (PP)"!(ComponentPose=@n22, Alpha=1, ConstraintWeights_0=0.6, ConstraintWeights_1:float=0.4) -> (Pose) {
    alphaBoolBlend {blendOption=Linear}
    boneToModify {boneName=dyn_backHairStrand_4_04}
    constraintSetup [{targetBone {boneName=spine_05}}, {targetBone {boneName=head}}]
    constraintWeights [1,1]
   }
   n24 = "Constraint - dyn_backHairStrand_5_03 (P)"(ComponentPose=@n23, Alpha=1, ConstraintWeights_0=1) {
    alphaBoolBlend {blendOption=Linear}
    boneToModify {boneName=dyn_backHairStrand_5_03}
    constraintSetup [{targetBone {boneName=head}}]
    constraintWeights [1]
   }
   n25 = "Constraint - dyn_backHairStrand_5_04 (PPP)"!(ComponentPose=@n24, Alpha=1, ConstraintWeights_0=0.6, ConstraintWeights_1:float=0.4, ConstraintWeights_2:float=0) -> (Pose) {
    alphaBoolBlend {blendOption=Linear}
    boneToModify {boneName=dyn_backHairStrand_5_04}
    constraintSetup [{targetBone {boneName=spine_05}}, {targetBone {boneName=head}}, {targetBone {boneName=clavicle_r}}]
    constraintWeights [1,1,1]
   }
   n26 = "Constraint - dyn_backHairStrand_5_05 (PP)"!(ComponentPose=@n25, Alpha=1, ConstraintWeights_0=0.5, ConstraintWeights_1:float=0.5) -> (Pose) {
    alphaBoolBlend {blendOption=Linear}
    boneToModify {boneName=dyn_backHairStrand_5_05}
    constraintSetup [{targetBone {boneName=spine_05}}, {targetBone {boneName=clavicle_r}}]
    constraintWeights [1,1]
   }
   n27 = "Constraint - dyn_backHairStrand_6_03 (PP)"!(ComponentPose=@n26, Alpha=1, ConstraintWeights_0=0.95, ConstraintWeights_1:float=0.05) -> (Pose) {
    alphaBoolBlend {blendOption=Linear}
    boneToModify {boneName=dyn_backHairStrand_6_03}
    constraintSetup [{targetBone {boneName=head}}, {targetBone {boneName=spine_05}}]
    constraintWeights [1,1]
   }
   n28 = "Constraint - dyn_backHairStrand_6_04 (PPP)"!(ComponentPose=@n27, Alpha=1, ConstraintWeights_0=0.7, ConstraintWeights_1:float=0.3, ConstraintWeights_2:float=0) -> (Pose) {
    alphaBoolBlend {blendOption=Linear}
    boneToModify {boneName=dyn_backHairStrand_6_04}
    constraintSetup [{targetBone {boneName=spine_05}}, {targetBone {boneName=head}}, {targetBone {boneName=clavicle_l}}]
    constraintWeights [1,1,1]
   }
   n29 = "Constraint - dyn_backHairStrand_6_05 (PP)"!(ComponentPose=@n28, Alpha=1, ConstraintWeights_0=0.5, ConstraintWeights_1:float=0.5) -> (Pose) {
    alphaBoolBlend {blendOption=Linear}
    boneToModify {boneName=dyn_backHairStrand_6_05}
    constraintSetup [{targetBone {boneName=spine_05}}, {targetBone {boneName=clavicle_l}}]
    constraintWeights [1,1]
   }
   Save cached pose 'Backpack Setup'(Pose=Component To Local(ComponentPose=@n29)) {cacheName=Backpack Setup}
  ]}
 EventGraph {nodes [
   Event BlueprintUpdateAnimation()
   > n1 = Sequence!(execute) -> (then_0, then_1, then_2:Exec)
   > "Set is Skydiving "("is Skydiving ":double="To Float (Boolean)"(InBool=Get bIsSkyDiving())) -> (Output_Get:double)
   > Set Neck Mover In/Out(Neck Mover In/Out:bool=OR Boolean(A=Get bIsInFrontEnd(), B=Get bIsOnGround() -> (bIsOnGround:bool))) -> (Output_Get:bool)
   > n2 = Branch(Condition=Get bIsSkyDiving())
   > Set GravityOverride(GravityOverride="0,0,500")
   > Set IsParachuteOpened(IsParachuteOpened:double="To Float (Boolean)"(InBool=Get bIsParachuteOpened())) -> (Output_Get:double)
   > Set IsSkydiveFloating(IsSkydiveFloating:double="To Float (Boolean)"(InBool=Get bIsSkydiveFloating())) -> (Output_Get:double)
   > n3 = Branch(Condition=Get bIsParachuteOpened())
   > Set GravityOverride(GravityOverride="0,0,-980")
   Set GravityOverride(execute=@n3.else, GravityOverride="0,0,-825")
   Set GravityOverride(execute=@n2.else, GravityOverride="0,0,-980")
   n4 = CheckForBackbling()
   Set is Wearing Backbling(
    execute=@n1.then_1
    is Wearing Backbling:bool=AND Boolean!(A=@n4, B=NOT Boolean(A=@n4.bIsBackblingACape), C:bool=NOT Boolean(A=@n4.bIsBackblingLong)) -> (ReturnValue)
   ) -> (Output_Get:bool)
   > Set is Wearing Cape(is Wearing Cape:bool=OR Boolean(A=@n4.bIsBackblingACape, B=@n4.bIsBackblingLong)) -> (Output_Get:bool)
   n5 = OR Boolean(A=Get is Wearing Backbling() -> (is Wearing Backbling:bool), B=Get is Wearing Cape() -> (is Wearing Cape:bool))
   > Set is Wearing Backbling or Cape(is Wearing Backbling or Cape:bool=@n5) -> (Output_Get:bool)
   > Set is Wearing Backbling or Cape Alpha(is Wearing Backbling or Cape Alpha:double="To Float (Boolean)"(InBool=@n5)) -> (Output_Get:double)
   n6 = Event OnCharacterCustomizationLoaded()
   > Set Body_Mesh(Body_Mesh=@n6.CustomBaseMesh)
   > Set Skel Mesh(Skel Mesh=Get Array Item with EFort Custom Part Type(Array=@n6.CustomMeshParts, PartType=Face))
  ]}
]
variables [
 Body_Mesh {type=SkeletalMeshComponent}
 Gravity Override Params {type=GravityOverrideParamsStruct}
 Gravity Value {type=Vector}
 GravityOverride {type=Vector}
 Hair Up {type=number}
 "is Skydiving " {type=number}
 is Wearing Backbling {type=boolean}
 is Wearing Backbling or Cape {type=boolean}
 is Wearing Backbling or Cape Alpha {type=number}
 is Wearing Cape {type=boolean}
 IsParachuteOpened {type=number}
 IsSkydiveFloating {type=number}
 Neck Mover In/Out {type=boolean}
 Skel Mesh {type=SkeletalMeshComponent}
]
